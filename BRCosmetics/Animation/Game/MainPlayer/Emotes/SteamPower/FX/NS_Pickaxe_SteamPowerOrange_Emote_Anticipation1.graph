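# Niagara effect for the anticipation part of the orange SteamPower pickaxe emote, pairing orange fire with icy blue accents.
# Emitters burst on user-set beat times, layering fire cross-planes, embers, flares, glows, lights and toon-shaded, outlined meshes.
# Particles sample User and System skeletal mesh sockets, with partitions or indices splitting some across two sockets.
# Time-of-day and front-end modifiers adjust brightness and size, while fire materials receive smoothed positions for motion lag.
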NiagaraSystem /BRCosmetics/Animation/Game/MainPlayer/Emotes/SteamPower/FX/NS_Pickaxe_SteamPowerOrange_Emote_Anticipation1
dependencies {
 usedDataInterfaces [
  &NiagaraDataInterfaceMeshRendererInfo
  &NiagaraDataInterfaceArrayFloat
  &NiagaraDataInterfacePlatformSet
  &NiagaraDataInterfaceCurve
  &NiagaraDataInterfaceVector2DCurve
  &NiagaraDataInterfaceColorCurve
  &NiagaraDataInterfaceVectorCurve
 ]
 usedDynamicInputs [&MakeFloatFromVector2d]
 usedModules [
  &self:SystemSpawnScript.SystemScriptSource.SystemScriptGraph.NiagaraNodeAssignment_3.SetVariables_#1
  &self:SystemSpawnScript.SystemScriptSource.SystemScriptGraph.NiagaraNodeAssignment_1.SetVariables_#2
  &SystemState
  &self:SystemSpawnScript.SystemScriptSource.SystemScriptGraph.NiagaraNodeAssignment_2.SetVariables_#3
 ]
}
emitters {
 AirGlow003 {
  emitterUpdateScript [
   &EmitterState()
   &SpawnBurst_Instantaneous(
    Spawn Count=2
    Spawn Time=&Add_Float(A=$User.SecondBeat, B=-0.13)
    Use Spawn Probability=false
    UseLoopCountLimit=true
   )
  ]
  particleSpawnScript [
   &PartitionParticles()
   &InitializeParticle(
    Color=&MultiplyLinearColor(
     Linear Color=&LerpLinearColors(
      StartColor=<rgba 1 0.15 0 1>
      EndColor=<rgba 0.125 0.5917 1 1>
      LerpFactor=&MakeFloatFromInt(Int=$Particles.PartitionParticles.Partition)
     )
     Scale Factor=<1 1 1 0.1>
    )
    Lifetime=0.7
    Sprite Size Mode=::NewEnumerator3
    Uniform Sprite Size=60.5513
    UsePositionOffset=false
    Write Material Random=true
   )
   &AddVelocity(Velocity=<1 0 0>, Velocity Speed Scale=10) {enabled=false}
  ]
  particleUpdateScript [
   &ParticleState()
   &SpriteRotationRate(Rotation Rate=400)
   &SkeletalMeshLocation(
    Bone / Socket Index=&ReturnExecIndex()
    Bone Sampling Mode=::NewEnumerator4
    MaskSpawnGroup=false
    Orientation Sampling=::NewEnumerator1
    Sampled Orientation Coordinate Space=::World
    Sampled Position Offset=&Lerp_Vector(
     A=<-3.8961 -2.0152 3.7698>
     B=<1.578 3.8611 3.1343>
     Alpha=&MakeFloatFromInt(Int=$Particles.PartitionParticles.Partition)
    )
    Sampled Position Offset Coordinate Space:ENiagaraSkelMeshTransforms=::NewEnumerator2
    Skeletal Mesh=$User.Skeletal Mesh
   ) {module=SkeletalMeshLocation001}
   &CameraOffset(Camera Offset Amount=30)
   /SetVariables(Particles.Velocity=&RotateVectorByQuaternion(Vector To Rotate=<0 0 1>, Quaternion=$Particles.SkeletalMeshLocation.SampledRotation)) {isSetParameters=true}
   &ScaleColor(
    Scale Alpha=&FloatFromCurve(FloatCurve=DataInterface() {propertyValues {Curve {keys [<at 0 0>,<at 0.3 1>,<at 0.9 1>,<at 1 0>]}}}, Scale Curve=8)
    Scale RGB=&VectorFromFloat(Value=&FNTimeOfDayScalar(Night Scale=0.5, Morning Scale=1, Evening Scale=0.8))
    ScaleA=true
    ScaleRGB=true
   )
   &ScaleColor(Scale Alpha=$System.FadeOut, ScaleA=true, ScaleRGB=true) {enabled=false, module=ScaleColor002}
   /SetVariables(
    Particles.Linear Color=&MultiplyLinearColor(
     Linear Color=<rgba 1 0.1 0 0.05>
     Scale Factor=&MakeVector4(
      X=1
      Y=1
      Z=1
      W=&FloatFromCurve(FloatCurve=DataInterface() {propertyValues {Curve {keys [<at 0.2 0>,<at 0.6 1>,<at 1 0>]}}}, Scale Curve=2)
     )
    )
    Particles.Size:Vector2f=&Vector2DFromFloat(Value=159.1815)
   ) {isSetParameters=true}
  ]
  rendererClasses [&NiagaraSpriteRendererProperties]
  renderers [
   {rendererClass=&NiagaraSpriteRendererProperties}
   {rendererClass=&NiagaraSpriteRendererProperties, rendererIndex=1}
  ]
 }
 Embers {
  emitterUpdateScript [
   &EmitterState()
   &SpawnBurst_Instantaneous(
    Spawn Count=10
    Spawn Time=&Add_Float(A=$User.ThirdBeat, B=&Subtract_Float(A=$User.LastBeatLifeTime, B=0.125))
    Use Spawn Probability=false
    UseLoopCountLimit=false
   )
   &SpawnRate(
    Spawn Probability=&MakeCustomFloatFromBool(Bool=$User.bisSwinging, True Float=0, False Float=0.5)
    SpawnRate=400
    Use Spawn Probability=true
   ) {enabled=false}
  ]
  particleSpawnScript [
   /SetVariables(Particles.PerParticleRandom:NiagaraFloat=&RandomRangeFloat(Minimum=0, Maximum=1)) {isSetParameters=true}
   &InitializeParticle(
    Color Mode=::NewEnumerator0
    Lifetime=&Lerp_Float(A=0.2, B=0.35, Alpha=$Particles.PerParticleRandom)
    Sprite Size=&Multiply_Vector2D(
     A=&Multiply_Vector2D_ByFloat(
      Float=&FN_FrontEndFloatModifier(Gameplay=1, Front End=&Multiply_Float(A=$User.FrontendScale, B=1))
      Vector2D=<1.5 1.3>
     )
     B=&MakeVector2D(
      X=&FloatFromCurve(
       FloatCurve=DataInterface() {propertyValues {Curve {keys [<at 0 1>,<at 0.8 2 RCIM_Constant>,<at 0.95 3 RCIM_Constant>]}}}
       CurveIndex=$Particles.PerParticleRandom
      )
      Y=&FloatFromCurve(
       FloatCurve=DataInterface() {propertyValues {Curve {keys [<at 0 1>,<at 0.8 2 RCIM_Constant>,<at 0.95 3>]}}}
       CurveIndex=$Particles.PerParticleRandom
      )
     )
    )
    Sprite Size Mode=::NewEnumerator4
    UsePositionOffset=false
    Write Material Random=true
   )
   &SkeletalMeshLocation(Bone Sampling Mode=::NewEnumerator4, MaskSpawnGroup=false, Sampled Position Offset=<-34.1987 0 -5.8832>, Skeletal Mesh=$System.Skeletal Mesh)
   &ShapeLocation(
    MaskSpawnGroup=false
    Offset Mode=::NewEnumerator0
    Override Seed=false
    Ring Radius=10
    Rotation Axis=<0 1 0>
    Rotation Mode=::NewEnumerator1
    Shape Primitive=::NewEnumerator4
    UseOwnerScale=true
   ) {module=ShapeLocation001}
   &AddVelocity(
    Cone Angle=90
    Cone Axis=<0.3 1 -0.37>
    Inner Cone Angle=90
    Module.UseConeFalloff=true
    Module.UseVelDistribution=false
    Override Seed=false
    Velocity Mode=::NewEnumerator2
    Velocity Speed=&Multiply_Float(
     A=&RandomRangeFloat(Minimum=150, Maximum=300)
     B=&FN_FrontEndFloatModifier(Gameplay=4, Front End=$User.FrontendScale)
    )
   ) {enabled=false, module=AddVelocity001}
   &AddVelocity(
    Additional Quaternion Rotation=true
    Additional Yaw / Pitch / Roll:Vector3f=<0 40 0>
    Cone Angle=60
    Inner Cone Angle=40
    Module.UseConeFalloff=false
    Module.UseVelDistribution=false
    Override Seed=false
    Rotation Mode=::NewEnumerator3
    Rotation Quaternion:Quat4f=$Particles.SkeletalMeshLocation.SampledRotation
    Velocity Mode=::NewEnumerator2
    Velocity Speed=&RandomRangeFloat(Minimum=50, Maximum=800)
   )
  ]
  particleUpdateScript [
   &ParticleState()
   &ScaleSpriteSize(
    Non-Uniform Curve Scale=<1 1.2>
    Non-Uniform Curve Sprite Scale=DataInterface() {propertyValues {
      XCurve {keys [{interpMode=RCIM_Cubic, tangentMode=RCTM_User, time=0, value=2}, {interpMode=RCIM_Cubic, tangentMode=RCTM_User, time=1, value=1}]}
      YCurve {keys [
        {arriveTangent=-25.5222, interpMode=RCIM_Cubic, leaveTangent=-25.5222, tangentMode=RCTM_User, time=0, value=10}
        {arriveTangent=-6.1872, interpMode=RCIM_Cubic, leaveTangent=-6.1872, tangentMode=RCTM_User, time=0.8, value=0}
       ]}
     }}
    Scale Sprite Size Mode=::NewEnumerator3
   ) {module=ScaleSpriteSize001}
   &Drag(
    Drag=&FloatFromCurve(FloatCurve=DataInterface() {propertyValues {Curve {keys [
         {arriveTangent=4.7857, interpMode=RCIM_Cubic, leaveTangent=4.7857, tangentMode=RCTM_User, time=0.5, value=7}
         {interpMode=RCIM_Cubic, tangentMode=RCTM_User, time=0.7, value=20}
        ]}}})
    Use Linear Drag=true
    Use Rotational Drag=false
   )
   &CurlNoiseForce(
    Noise Strength=&FloatFromCurve(
     FloatCurve=DataInterface() {propertyValues {Curve {keys [{interpMode=RCIM_Cubic, tangentMode=RCTM_User, time=0.12, value=0}, {interpMode=RCIM_Cubic, tangentMode=RCTM_User, time=0.3, value=1}]}}}
     Scale Curve=4000
    )
    Pan Noise=false
    Random Seed=33
   ) {module=CurlNoiseForce001}
   &InheritSourceMovement(
    Applied Position Scale=&VectorFromFloat(Value=&FloatFromCurve(FloatCurve=DataInterface() {propertyValues {Curve {keys [{interpMode=RCIM_Cubic, tangentMode=RCTM_User, time=0.5, value=1}, {interpMode=RCIM_Cubic, tangentMode=RCTM_User, time=1, value=0}]}}}))
    ApplyForce=false
    ApplyPosition=true
    ApplyPositionFalloff=false
   )
   &GravityForce(Gravity=&MakeVector(
     X=0
     Y=0
     Z=&FloatFromCurve(
      FloatCurve=DataInterface() {propertyValues {Curve {keys [
          {arriveTangent=0.2564, interpMode=RCIM_Cubic, leaveTangent=0.2564, tangentMode=RCTM_User, time=0.2, value=1}
          {arriveTangent=1.4946, interpMode=RCIM_Cubic, leaveTangent=1.4946, tangentMode=RCTM_User, time=0.5, value=2}
         ]}}}
      Scale Curve=900
     )
    ))
   &SolveForcesAndVelocity(Clamp Velocity=false, Limit Acceleration=false)
   &ScaleColor(
    Scale RGB=&MakeVectorFromLinearColorRGB(Color=&ColorFromCurve(ColorCurve=DataInterface() {propertyValues {
        AlphaCurve {keys [<at 0 1 RCIM_Cubic>]}
        BlueCurve {keys [<at 0.2182 0 RCIM_Cubic>,<at 0.5226 0>,<at 1 0 RCIM_Cubic>]}
        GreenCurve {keys [<at 0.2182 0.1667 RCIM_Cubic>,<at 0.5226 0>,<at 1 0 RCIM_Cubic>]}
        RedCurve {keys [<at 0.2182 5 RCIM_Cubic>,<at 0.5226 1>,<at 1 0.5 RCIM_Cubic>]}
       }}))
    ScaleA=true
    ScaleRGB=true
   )
   &ScaleSpriteSize(
    Uniform Curve Scale=&MakeFloatFromVector(Vector=$Engine.Owner.Scale, Channel=::NewEnumerator0)
    Uniform Curve Sprite Scale=DataInterface() {propertyValues {Curve {keys [
        {tangentMode=RCTM_User, time=0, value=1}
        {arriveTangent=-0.0405, interpMode=RCIM_Cubic, leaveTangent=-0.0405, tangentMode=RCTM_User, time=0.7009, value=1}
        {arriveTangent=-8.0089, interpMode=RCIM_Cubic, leaveTangent=-8.0089, tangentMode=RCTM_User, time=1, value=0}
       ]}}}
   )
   &ScaleSpriteSizeBySpeed(Max Scale Factor=<2 7>, Velocity Threshold=282.2583) {enabled=false}
   &ScaleColor(
    Scale Alpha=&Lerp_Float(A=$System.EmoteIntro, B=$System.EmoteOutro, Alpha=$User.AnimState)
    ScaleA=true
    ScaleRGB=true
   ) {module=ScaleColor002}
  ]
  rendererClasses [&NiagaraSpriteRendererProperties, &NiagaraMeshRendererProperties]
  renderers [
   {rendererClass=&NiagaraSpriteRendererProperties}
   {rendererClass=&NiagaraMeshRendererProperties, rendererIndex=1}
  ]
 }
 Fire_Mesh_Cross_Plane {
  emitterUpdateScript [
   &EmitterState()
   &SpawnBurst_Instantaneous(Spawn Count=2, Spawn Time=&Add_Float(A=$User.FirstBeat, B=0), Use Spawn Probability=false, UseLoopCountLimit=false)
  ]
  particleSpawnScript [
   &InitializeParticle(
    Lifetime=0.9
    Mesh Scale=&Multiply_VectorByFloat(Vector=<1 1.1 1>, Float=0.22)
    Mesh Scale Mode=::NewEnumerator4
    Position Offset=<0 0 -10>
    Position Offset Coordinate Space=::Local
    UsePositionOffset=true
    Write Material Random=true
   )
   &FN_SetSpawnGroup(Number of Spawn Groups=2)
   &InitialMeshOrientation(
    Mesh Axis To Orient=<0 -1 0>
    Mesh Orientation Mode=::NewEnumerator2
    Rotation=&MakeVector(
     X=0
     Y=&SelectFloatFromArray(Direct Array Index=$Particles.SpawnGroup, Float Selection Array=DataInterface() {propertyValues {FloatData [0,0.25]}})
     Z=0
    )
    Use Rotation Vector=true
   )
   &DynamicMaterialParameters(
    Index 0 Param 1=&SelectFloatFromArray(Direct Array Index=$Particles.SpawnGroup, Float Selection Array=DataInterface() {propertyValues {FloatData [0,0.3]}})
    Index 2 Param 1=&SelectFloatFromArray(Direct Array Index=$Engine.QualityLevel, Float Selection Array=DataInterface() {propertyValues {FloatData [0.4,0.4,1,1,1]}})
    Module.Param0WriteEnabled=true
    Module.Param0WriteEnabled2=true
    Module.Param1WriteEnabled=true
    Module.Param1WriteEnabled2=true
    Module.Param2WriteEnabled=true
    Module.Param2WriteEnabled2=true
    Module.Param3WriteEnabled=true
    Module.Param3WriteEnabled2=true
    Write Parameter Index 2=true
   )
   &PartitionParticles(Number Of Partitions=5)
  ]
  particleUpdateScript [
   &ParticleState()
   &SkeletalMeshLocation(Bone Sampling Mode=::NewEnumerator4, MaskSpawnGroup=false, Sampled Position Offset=<-12.7477 0 0>, Skeletal Mesh=$User.Skeletal Mesh)
   &InterpolateOverTime(Rate Of Change=10, Target Vector Value:NiagaraPosition=$Particles.Position, Variable Type=::NewEnumerator2)
   &DynamicMaterialParameters(
    Index 1 Param 1=&MakeFloatFromVector(Vector=$Particles.InterpolateOverTime.InterpolateOverTime_V3_.Moving Average, Channel=::NewEnumerator0)
    Index 1 Param 2=&MakeFloatFromVector(Vector=$Particles.InterpolateOverTime.InterpolateOverTime_V3_.Moving Average, Channel=::NewEnumerator1)
    Index 1 Param 3=&MakeFloatFromVector(Vector=$Particles.InterpolateOverTime.InterpolateOverTime_V3_.Moving Average, Channel=::NewEnumerator2)
    Module.Param0WriteEnabled1=true
    Module.Param1WriteEnabled1=true
    Module.Param2WriteEnabled1=true
    Module.Param3WriteEnabled1=true
    Write Parameter Index 0=false
    Write Parameter Index 1=true
   ) {module=DynamicMaterialParameters001}
   &ScaleColor(
    Scale Alpha=&FloatFromCurve(FloatCurve=DataInterface() {propertyValues {Curve {keys [<at 0 0>,<at 0.3 0.7>,<at 0.9 0.7>,<at 1 0>]}}}, Scale Curve=1.3)
    Scale RGB=&VectorFromFloat(Value=&Multiply_Float(
      A=&FN_FrontEndFloatModifier(Gameplay=1, Front End=1.1)
      B=&Multiply_Float(
       A=&MakeCustomFloatFromBool(
        Bool=&IsPlatformSetActive(New Platform Set=DataInterface() {propertyValues {Platforms {qualityLevelMask=-29}}})
        True Float=3
        False Float=1
       )
       B=&FNTimeOfDayScalar(Night Scale=0.5, Morning Scale=0.75, Evening Scale=0.75)
      )
     ))
    ScaleA=true
    ScaleRGB=true
   )
   &ScaleColor(
    Color Value To Scale=$Particles.Color
    Scale Alpha=&FN_FrontEndFloatModifier(Gameplay=1, Front End=0.9)
    Scale RGB=&FN_FrontEndVector3Modifier(Gameplay=<1 1 1>, Front End=&MakeVectorFromLinearColorRGB(Color=<rgba 1 0.4771 0.408 1>))
    ScaleA=true
    ScaleRGB=true
   ) {enabled=false, module=ScaleColor001}
   &ScaleColor(
    Scale Alpha=&Lerp_Float(A=$System.EmoteIntro, B=$System.EmoteOutro, Alpha=$User.AnimState)
    ScaleA=true
    ScaleRGB=true
   ) {enabled=false, module=ScaleColor002}
  ]
  rendererClasses [&NiagaraMeshRendererProperties, &NiagaraSpriteRendererProperties]
  renderers [
   {rendererClass=&NiagaraMeshRendererProperties, rendererIndex=0}
   {rendererClass=&NiagaraSpriteRendererProperties, rendererIndex=1}
  ]
 }
 Fire_Mesh_Cross_Plane001 {
  emitterUpdateScript [
   &EmitterState()
   &SpawnBurst_Instantaneous(Spawn Count=6, Spawn Time=&Add_Float(A=$User.ThirdBeat, B=0), Use Spawn Probability=false, UseLoopCountLimit=false)
  ]
  particleSpawnScript [
   &InitializeParticle(
    Lifetime=$User.LastBeatLifeTime
    Mesh Scale=&Multiply_VectorByFloat(Vector=<1.2 2.2 1>, Float=0.22)
    Mesh Scale Mode=::NewEnumerator4
    Position Offset=<0 0 -10>
    Position Offset Coordinate Space=::Local
    UsePositionOffset=true
    Write Material Random=true
   )
   &FN_SetSpawnGroup(Number of Spawn Groups=2)
   &InitialMeshOrientation(
    Mesh Axis To Orient=<0.3796 -1 0>
    Mesh Orientation Mode=::NewEnumerator2
    Rotation=&MakeVector(
     X=0
     Y=&SelectFloatFromArray(Direct Array Index=$Particles.SpawnGroup, Float Selection Array=DataInterface() {propertyValues {FloatData [0,0.25]}})
     Z=0
    )
    Use Rotation Vector=true
   )
   &DynamicMaterialParameters(
    Index 0 Param 1=&SelectFloatFromArray(Direct Array Index=$Particles.SpawnGroup, Float Selection Array=DataInterface() {propertyValues {FloatData [0,0.3]}})
    Index 2 Param 1=&SelectFloatFromArray(Direct Array Index=$Engine.QualityLevel, Float Selection Array=DataInterface() {propertyValues {FloatData [0.4,0.4,1,1,1]}})
    Module.Param0WriteEnabled=true
    Module.Param0WriteEnabled2=true
    Module.Param1WriteEnabled=true
    Module.Param1WriteEnabled2=true
    Module.Param2WriteEnabled=true
    Module.Param2WriteEnabled2=true
    Module.Param3WriteEnabled=true
    Module.Param3WriteEnabled2=true
    Write Parameter Index 2=true
   )
   &PartitionParticles(Number Of Partitions=5)
  ]
  particleUpdateScript [
   &ParticleState()
   &ScaleMeshSize(Scale Factor=<1.5 1.1 1.5>) {module=ScaleMeshSize001}
   &ScaleMeshSize(Scale Factor=&MakeVector(
     X=&FloatFromCurve(FloatCurve=DataInterface() {propertyValues {Curve {keys [<at 0 1>,<at 1 1.2>]}}})
     Y=&FloatFromCurve(FloatCurve=DataInterface() {propertyValues {Curve {keys [<at 0 1>,<at 1 1>]}}})
     Z=&FloatFromCurve(FloatCurve=DataInterface() {propertyValues {Curve {keys [<at 0 1>,<at 1 1.2>]}}})
    )) {module=ScaleMeshSize002}
   &ScaleMeshSize(Scale Factor=&VectorFromFloat(Value=&FloatFromCurve(
      FloatCurve=DataInterface() {propertyValues {Curve {keys [{interpMode=RCIM_Cubic, tangentMode=RCTM_User, time=0, value=0.5}, {interpMode=RCIM_Cubic, tangentMode=RCTM_User, time=0.8, value=0.9}]}}}
      Scale Curve=1.2019
     ))) {module=ScaleMeshSize003}
   &ScaleMeshSize(Scale Factor=&MakeVector(
     X=&FloatFromCurve(FloatCurve=DataInterface() {propertyValues {Curve {keys [
          {interpMode=RCIM_Cubic, tangentMode=RCTM_User, time=0, value=1}
          {interpMode=RCIM_Cubic, tangentMode=RCTM_User, time=0.3, value=0.2}
          {interpMode=RCIM_Cubic, tangentMode=RCTM_User, time=1, value=0}
         ]}}})
     Y=&FloatFromCurve(FloatCurve=DataInterface() {propertyValues {Curve {keys [
          {interpMode=RCIM_Cubic, tangentMode=RCTM_User, time=0, value=1}
          {interpMode=RCIM_Cubic, tangentMode=RCTM_User, time=0.3, value=0.2}
          {interpMode=RCIM_Cubic, tangentMode=RCTM_User, time=1, value=0}
         ]}}})
     Z=&FloatFromCurve(FloatCurve=DataInterface() {propertyValues {Curve {keys [<at 0 1>,<at 1 0>]}}})
    )) {enabled=false}
   &SkeletalMeshLocation(Bone Sampling Mode=::NewEnumerator4, MaskSpawnGroup=false, Sampled Position Offset=<-25.231 10.1073 0>, Skeletal Mesh=$User.Skeletal Mesh)
   &InterpolateOverTime(Rate Of Change=10, Target Vector Value:NiagaraPosition=$Particles.Position, Variable Type=::NewEnumerator2)
   &DynamicMaterialParameters(
    Index 1 Param 1=&MakeFloatFromVector(Vector=$Particles.InterpolateOverTime.InterpolateOverTime_V3_.Moving Average, Channel=::NewEnumerator0)
    Index 1 Param 2=&MakeFloatFromVector(Vector=$Particles.InterpolateOverTime.InterpolateOverTime_V3_.Moving Average, Channel=::NewEnumerator1)
    Index 1 Param 3=&MakeFloatFromVector(Vector=$Particles.InterpolateOverTime.InterpolateOverTime_V3_.Moving Average, Channel=::NewEnumerator2)
    Module.Param0WriteEnabled1=true
    Module.Param1WriteEnabled1=true
    Module.Param2WriteEnabled1=true
    Module.Param3WriteEnabled1=true
    Write Parameter Index 0=false
    Write Parameter Index 1=true
   ) {module=DynamicMaterialParameters001}
   &ScaleColor(
    Scale RGB=&VectorFromFloat(Value=&FNTimeOfDayScalar(Night Scale=0.5, Morning Scale=0.75, Evening Scale=0.75))
    ScaleA=true
    ScaleRGB=true
   )
   &ScaleColor(
    Color Value To Scale=$Particles.Color
    Scale Alpha=&FN_FrontEndFloatModifier(Gameplay=1, Front End=0.9)
    Scale RGB=&FN_FrontEndVector3Modifier(Gameplay=<1 1 1>, Front End=&MakeVectorFromLinearColorRGB(Color=<rgba 1 0.4771 0.408 1>))
    ScaleA=true
    ScaleRGB=true
   ) {module=ScaleColor001}
   &ScaleColor(
    Scale Alpha=&Lerp_Float(A=$System.EmoteIntro, B=$System.EmoteOutro, Alpha=$User.AnimState)
    ScaleA=true
    ScaleRGB=true
   ) {enabled=false, module=ScaleColor002}
   &ScaleColor(
    Scale Alpha=&FloatFromCurve(FloatCurve=DataInterface() {propertyValues {Curve {keys [
         {interpMode=RCIM_Cubic, tangentMode=RCTM_User, time=0.5, value=1}
         {arriveTangent=-0.1344, interpMode=RCIM_Cubic, leaveTangent=-0.1344, tangentMode=RCTM_User, time=1, value=0}
        ]}}})
    ScaleA=true
    ScaleRGB=true
   ) {module=ScaleColor003}
   &ScaleColor(
    Scale RGB=&VectorFromFloat(Value=&FloatFromCurve(FloatCurve=DataInterface() {propertyValues {Curve {keys [
          {interpMode=RCIM_Cubic, tangentMode=RCTM_User, time=0, value=3}
          {arriveTangent=-6.2946, interpMode=RCIM_Cubic, leaveTangent=-6.2946, tangentMode=RCTM_User, time=0.45, value=1}
          {interpMode=RCIM_Cubic, tangentMode=RCTM_User, time=0.6, value=1}
         ]}}}))
    ScaleA=true
    ScaleRGB=true
   ) {module=ScaleColor004}
   &FN_Light_Attributes(
    Color=<rgba 4 1.0793 0.336 1>
    Create Color Attribute=true
    Create Position Attribute=true
    Position=&AddVectorToPosition(
     Position=$Particles.Position
     Vector=&RotateVectorByQuaternion(Vector To Rotate=<-18.5717 1.9228 0>, Quaternion=$Particles.MeshOrientation)
    )
    Scale Color=&FloatFromCurve(FloatCurve=DataInterface() {propertyValues {Curve {keys [
         {arriveTangent=8.8782, interpMode=RCIM_Cubic, leaveTangent=8.8782, tangentMode=RCTM_User, time=0, value=0}
         {interpMode=RCIM_Cubic, tangentMode=RCTM_User, time=0.1, value=1}
         {arriveTangent=-1.7003, interpMode=RCIM_Cubic, leaveTangent=-1.7003, tangentMode=RCTM_User, time=0.8, value=0}
        ]}}})
   )
   /SetVariables(
    Particles.FlareColor=&MultiplyLinearColor(
     Linear Color=<rgba 20 2.6667 0 0.007>
     Scale Factor=&MakeFloat4FromVectorAndFloat(
      "Vector (RGB)"=&VectorFromFloat(Value=&FloatFromCurve(
        FloatCurve=DataInterface() {propertyValues {Curve {keys [
            {interpMode=RCIM_Cubic, tangentMode=RCTM_User, time=0, value=1}
            <at 1 1>
            {arriveTangent=-3.2575, interpMode=RCIM_Cubic, leaveTangent=-3.2575, tangentMode=RCTM_User, time=1, value=0}
           ]}}}
        Scale Curve=4
       ))
      "Float (Alpha)"=&FloatFromCurve(
       FloatCurve=DataInterface() {propertyValues {Curve {keys [
           {interpMode=RCIM_Cubic, tangentMode=RCTM_User, time=0, value=1}
           {arriveTangent=-2.6423, interpMode=RCIM_Cubic, leaveTangent=-2.6423, tangentMode=RCTM_User, time=1, value=1}
          ]}}}
       Scale Curve=0.05
      )
     )
    )
    Particles.FlarePosition:NiagaraPosition=&AddVectorToPosition(
     Position=$Particles.SkeletalMeshLocation.SampledPosition
     Vector=&RotateVectorByQuaternion(Vector To Rotate=<6.172 -0.5117 0>, Quaternion=$Particles.SkeletalMeshLocation.SampledRotation)
    )
    Particles.FlareSize:Vector2f=&Multiply_Vector2D_ByFloat(Float=0.7, Vector2D=<61.848 215.2872>)
    Particles.FlareVelocity:Vector3f=&RotateVectorByQuaternion(Vector To Rotate=<1 -0.3796 0>, Quaternion=$Particles.SkeletalMeshLocation.SampledRotation)
   ) {isSetParameters=true}
  ]
  rendererClasses [
   &NiagaraMeshRendererProperties
   &NiagaraLightRendererProperties
   &NiagaraSpriteRendererProperties
  ]
  renderers [
   {rendererClass=&NiagaraMeshRendererProperties, rendererIndex=0}
   {rendererClass=&NiagaraMeshRendererProperties, rendererIndex=1}
   {rendererClass=&NiagaraLightRendererProperties, rendererIndex=2}
   {rendererClass=&NiagaraSpriteRendererProperties, rendererIndex=3}
  ]
 }
 Flare_Glow {
  emitterSpawnScript [/SetVariables(Emitter.SizingCurve=DataInterface() {propertyValues {Curve {keys [
        {arriveTangent=60.4226, interpMode=RCIM_Cubic, leaveTangent=60.4226, tangentMode=RCTM_User, time=0, value=0}
        {arriveTangentWeight=0.175, interpMode=RCIM_Cubic, leaveTangentWeight=0.1588, tangentMode=RCTM_User, time=0.05, value=1.8822}
        {arriveTangentWeight=0.1341, interpMode=RCIM_Cubic, leaveTangentWeight=0.2413, tangentMode=RCTM_User, time=0.2592, value=1}
        {arriveTangentWeight=0.1593, interpMode=RCIM_Cubic, leaveTangentWeight=0.0885, tangentMode=RCTM_User, time=0.4464, value=1}
        {arriveTangentWeight=0.0833, interpMode=RCIM_Cubic, leaveTangent=-7.6948, time=0.593, value=0}
       ]}}}) {isSetParameters=true}]
  emitterUpdateScript [
   &EmitterState(
    Inactive Response=::NewEnumerator1
    Life Cycle Mode=::NewEnumerator1
    Loop Behavior=::NewEnumerator1
    Loop Delay=&Add_Float(A=$System.GlobalDelay, B=-0.05)
    Loop Duration=4
    UseLoopDelay=true
   )
   &SpawnBurst_Instantaneous(
    Spawn Count=2
    Spawn Time=&Add_Float(A=$User.SecondBeat, B=0)
    Use Spawn Probability=false
    UseLoopCountLimit=false
   )
  ]
  particleSpawnScript [
   &PartitionParticles()
   &InitializeParticle(
    Color=&MultiplyLinearColor(
     Linear Color=&LerpLinearColors(
      StartColor=<rgba 4 1.2667 0 1>
      EndColor=<rgba 0.64 1.3653 2 1>
      LerpFactor=&MakeFloatFromInt(Int=$Particles.PartitionParticles.Partition)
     )
     Scale Factor=<1 1 1 1>
    )
    Lifetime=0.8
    Mesh Scale Mode=::NewEnumerator3
    Mesh Uniform Scale=&FN_FrontEndFloatModifier(Gameplay=4, Front End=3.2)
    Sprite Rotation Mode=::NewEnumerator2
    Sprite Size Mode=::NewEnumerator3
    Uniform Sprite Size=&FN_FrontEndFloatModifier(Gameplay=300, Front End=250)
    UsePositionOffset=true
    Write Material Random=true
   )
   &InitialMeshOrientation(Mesh Orientation Mode=::NewEnumerator7, Rotation=<0 -0.25 0>, Use Rotation Vector=true)
   &InitialMeshOrientation(
    Mesh Orientation Mode=::NewEnumerator7
    Rotation=&RandomRangeVector(Minimum=<0 0 0>, Maximum=<1 0 0>)
    Use Rotation Vector=true
   ) {module=InitialMeshOrientation001}
   &CameraOffset(Camera Offset Amount=&FN_FrontEndFloatModifier(Gameplay=0, Front End=100))
  ]
  particleUpdateScript [
   &ParticleState()
   &SkeletalMeshLocation(
    Bone / Socket Index=$Particles.PartitionParticles.Partition
    Bone Sampling Mode=::NewEnumerator4
    MaskSpawnGroup=false
    Sampled Orientation Coordinate Space=::World
    Sampled Position Offset=&Lerp_Vector(
     A=<-3.8961 -2.0152 3.7698>
     B=<3.7689 3.7752 6.859>
     Alpha=&MakeFloatFromInt(Int=$Particles.PartitionParticles.Partition)
    )
    Sampled Position Offset Coordinate Space:ENiagaraSkelMeshTransforms=::NewEnumerator2
    Skeletal Mesh=$User.Skeletal Mesh
   ) {module=SkeletalMeshLocation001}
   &UpdateMeshOrientation(Rotation Rate=2, Rotation Vector=<0 0 1>) {enabled=false}
   &ScaleColor(
    Scale Alpha=&Multiply_Float(
     A=&Lerp_Float(A=0.75, B=1, Alpha=&Sine(Normalized Angle=$Emitter.Age, Period=0.5, Scale=0.5, Bias=0.5))
     B=&FloatFromCurve(FloatCurve=DataInterface() {propertyValues {Curve {keys [<at 0.95 1>,<at 1 0.1>]}}})
    )
    Scale RGB=&VectorFromFloat(Value=&Multiply_Float(
      A=&FloatFromCurve(FloatCurve=$Emitter.SizingCurve, Scale Curve=0.1)
      B=&FNTimeOfDayScalar(Night Scale=0.5, Morning Scale=1, Evening Scale=1)
     ))
    ScaleA=true
    ScaleRGB=true
   ) {module=ScaleColor001}
   &ScaleMeshSize(Scale Factor=&VectorFromFloat(Value=&FloatFromCurve(
      FloatCurve=DataInterface() {propertyValues {Curve {keys [
          {interpMode=RCIM_Cubic, tangentMode=RCTM_User, time=0, value=0}
          {interpMode=RCIM_Cubic, tangentMode=RCTM_User, time=0.2, value=1.3}
          {interpMode=RCIM_Cubic, tangentMode=RCTM_User, time=0.8, value=0}
         ]}}}
      Scale Curve=0.1773
     )))
   &ScaleSpriteSize(Uniform Curve Scale=0.5, Uniform Curve Sprite Scale:NiagaraDataInterfaceCurve=$Emitter.SizingCurve)
   &FN_Light_Attributes(
    Color=$Particles.Color
    Create Color Attribute=true
    Create Exponent Attribute=true
    Create Radius Attribute=true
    Exponent=2
    Radius=300
    Scale Color=4
   )
   &ScaleColor(
    Scale Alpha=&FloatFromCurve(
     FloatCurve=DataInterface() {propertyValues {Curve {keys [
         {interpMode=RCIM_Cubic, tangentMode=RCTM_User, time=0, value=0}
         {arriveTangent=4.9496, interpMode=RCIM_Cubic, leaveTangent=0.3212, tangentMode=RCTM_Break, time=0.4, value=1}
         {interpMode=RCIM_Cubic, tangentMode=RCTM_User, time=0.9, value=1}
         {interpMode=RCIM_Cubic, tangentMode=RCTM_User, time=1, value=0}
        ]}}}
     Scale Curve=0.15
    )
    Scale RGB=&VectorFromFloat(Value=&FNTimeOfDayScalar(Night Scale=0.5, Morning Scale=1, Evening Scale=0.8))
    ScaleA=true
    ScaleRGB=true
   )
   /SetVariables(
    Particles.Corona Cam Offset=20
    Particles.Corona Color:LinearColor=&MultiplyLinearColor(Linear Color=$Particles.Color, Scale Factor=&MakeVector4_FromFloat(Float=8))
    Particles.Corona Size:Vector2f=&Multiply_Vector2D_ByFloat(Float=2.2332, Vector2D=<100 100>)
    Particles.Vector 4:Vector4f=<0.1 0 0 0>
   ) {isSetParameters=true}
   &DynamicMaterialParameters(
    Index 0 Param 2=&FloatFromCurve(FloatCurve=DataInterface() {propertyValues {Curve {keys [<at 0 0.1>,<at 1 0>]}}}, Scale Curve=0.003)
    Index 0 Param 3=1
    Module.Index0Param0Write=true
    Module.Index0Param1Write=true
    Module.Index0Param2Write=true
    Module.Index0Param3Write=true
   )
  ]
  rendererClasses [
   &NiagaraMeshRendererProperties
   &NiagaraSpriteRendererProperties
   &NiagaraLightRendererProperties
  ]
  renderers [
   {rendererClass=&NiagaraMeshRendererProperties, rendererIndex=0}
   {rendererClass=&NiagaraSpriteRendererProperties, rendererIndex=1}
   {rendererClass=&NiagaraLightRendererProperties, rendererIndex=2}
   {rendererClass=&NiagaraSpriteRendererProperties, rendererIndex=3}
  ]
 }
 Flare_and_Light {
  emitterUpdateScript [
   &EmitterState()
   &SpawnBurst_Instantaneous(Spawn Time=&Add_Float(A=$User.ThirdBeat, B=0), Use Spawn Probability=false, UseLoopCountLimit=false)
  ]
  particleSpawnScript [&InitializeParticle(
    Color=<rgba 20 2.6667 0 1>
    Lifetime=0.07
    Sprite Size=<90 164.2674>
    Sprite Size Mode=::NewEnumerator4
    UsePositionOffset=false
    Write Material Random=true
   )]
  particleUpdateScript [
   &ParticleState()
   &SkeletalMeshLocation(Bone Sampling Mode=::NewEnumerator4, MaskSpawnGroup=false, Sampled Position Offset=<-12.8728 2.462 0>, Skeletal Mesh=$System.Skeletal Mesh)
   &ScaleColor(
    Scale Alpha=&FN_FrontEndFloatModifier(Gameplay=0.5, Front End=0.5)
    Scale RGB=&Multiply_Vector(
     A=&VectorFromFloat(Value=&FNTimeOfDayScalar(Night Scale=0.5, Morning Scale=0.75, Evening Scale=0.75))
     B=&VectorFromFloat(Value=&FloatFromCurve(FloatCurve=DataInterface() {propertyValues {Curve {keys [<at 0 1>,<at 1 0>]}}}))
    )
    ScaleA=true
    ScaleRGB=true
   ) {module=ScaleColor001}
   &ScaleColor(
    Scale Alpha=&Lerp_Float(A=$System.EmoteIntro, B=$System.EmoteOutro, Alpha=$User.AnimState)
    ScaleA=true
    ScaleRGB=true
   ) {module=ScaleColor002}
   /SetVariables(Particles.Velocity=&RotateVectorByQuaternion(Vector To Rotate=<1 0 0>, Quaternion=$Particles.SkeletalMeshLocation.SampledRotation)) {isSetParameters=true, module=SetVariables_#4}
   &FN_Light_Attributes(
    Color=<rgba 8 2.1585 0.672 1>
    Create Color Attribute=true
    Scale Color=&FloatFromCurve(FloatCurve=DataInterface() {propertyValues {Curve {keys [
         {interpMode=RCIM_Cubic, tangentMode=RCTM_User, time=0, value=1}
         {arriveTangent=-1.7003, interpMode=RCIM_Cubic, leaveTangent=-1.7003, tangentMode=RCTM_User, time=1, value=0}
        ]}}})
   )
  ]
  rendererClasses [&NiagaraMeshRendererProperties, &NiagaraSpriteRendererProperties]
  renderers [
   {rendererClass=&NiagaraMeshRendererProperties, rendererIndex=0}
   {rendererClass=&NiagaraSpriteRendererProperties, rendererIndex=1}
  ]
 }
 Flare_and_Light001 {
  emitterUpdateScript [
   &EmitterState()
   &SpawnBurst_Instantaneous(Spawn Time=&Add_Float(A=$User.FirstBeat, B=0), Use Spawn Probability=false, UseLoopCountLimit=false)
  ]
  particleSpawnScript [&InitializeParticle(
    Color=<rgba 20 2.6667 0 1>
    Lifetime=0.9
    Sprite Size=<70 70>
    Sprite Size Mode=::NewEnumerator4
    UsePositionOffset=false
    Write Material Random=true
   )]
  particleUpdateScript [
   &ParticleState()
   &SkeletalMeshLocation(Bone Sampling Mode=::NewEnumerator4, MaskSpawnGroup=false, Sampled Position Offset=<-12.8728 2.462 0>, Skeletal Mesh=$System.Skeletal Mesh) {enabled=false}
   &SkeletalMeshLocation(
    Bone / Socket Index=&ReturnExecIndex()
    Bone Sampling Mode=::NewEnumerator4
    MaskSpawnGroup=false
    Orientation Sampling=::NewEnumerator1
    Sampled Orientation Coordinate Space=::World
    Sampled Position Offset=&Lerp_Vector(
     A=<-3.8961 -2.0152 3.7698>
     B=<1.578 3.8611 3.1343>
     Alpha=&MakeFloatFromInt(Int=&ReturnExecIndex())
    )
    Sampled Position Offset Coordinate Space:ENiagaraSkelMeshTransforms=::NewEnumerator2
    Skeletal Mesh=$System.Skeletal Mesh
   ) {module=SkeletalMeshLocation001}
   &ScaleColor(
    Scale Alpha=0.5
    Scale RGB=&Multiply_Vector(
     A=&VectorFromFloat(Value=&FNTimeOfDayScalar(Night Scale=0.5, Morning Scale=0.75, Evening Scale=0.75))
     B=&VectorFromFloat(Value=&FloatFromCurve(FloatCurve=DataInterface() {propertyValues {Curve {keys [<at 0 1>,<at 1 0>]}}}))
    )
    ScaleA=true
    ScaleRGB=true
   ) {enabled=false, module=ScaleColor001}
   &ScaleColor(
    Scale Alpha=&Lerp_Float(A=$System.EmoteIntro, B=$System.EmoteOutro, Alpha=$User.AnimState)
    ScaleA=true
    ScaleRGB=true
   ) {module=ScaleColor002}
   /SetVariables(Particles.Velocity=&RotateVectorByQuaternion(Vector To Rotate=<1 0 0>, Quaternion=$Particles.SkeletalMeshLocation.SampledRotation)) {isSetParameters=true, module=SetVariables_#4}
   &ScaleColor(
    Scale Alpha=&FloatFromCurve(FloatCurve=DataInterface() {propertyValues {Curve {keys [<at 0 0>,<at 0.3 0.7>,<at 0.9 0.7>,<at 1 0>]}}}, Scale Curve=0.04)
    Scale RGB=&VectorFromFloat(Value=&Multiply_Float(
      A=&FN_FrontEndFloatModifier(Gameplay=1, Front End=1.1)
      B=&Multiply_Float(
       A=&MakeCustomFloatFromBool(
        Bool=&IsPlatformSetActive(New Platform Set=DataInterface() {propertyValues {Platforms {qualityLevelMask=-29}}})
        True Float=3
        False Float=1
       )
       B=&FNTimeOfDayScalar(Night Scale=0.5, Morning Scale=0.75, Evening Scale=0.75)
      )
     ))
    ScaleA=true
    ScaleRGB=true
   )
   &FN_Light_Attributes(
    Color=<rgba 8 2.1585 0.672 1>
    Create Color Attribute=true
    Scale Color=&FloatFromCurve(FloatCurve=DataInterface() {propertyValues {Curve {keys [
         {interpMode=RCIM_Cubic, tangentMode=RCTM_User, time=0, value=1}
         {arriveTangent=-1.7003, interpMode=RCIM_Cubic, leaveTangent=-1.7003, tangentMode=RCTM_User, time=1, value=0}
        ]}}})
   ) {enabled=false}
  ]
  rendererClasses [&NiagaraMeshRendererProperties, &NiagaraSpriteRendererProperties]
  renderers [
   {rendererClass=&NiagaraMeshRendererProperties, rendererIndex=0}
   {rendererClass=&NiagaraSpriteRendererProperties, rendererIndex=1}
  ]
 }
 NE_ToonShadingMesh {
  emitterSpawnScript [
   /Game/Effects/Niagara/MasterMaterials/ToonShader/Template/NM_ToonShaderParameters()
   /Game/Effects/Niagara/MasterMaterials/ToonShader/Template/NM_OutlineParameters(OutlineColor:LinearColor=<rgba 0 0.0601 0.224 1>)
  ]
  emitterUpdateScript [
   &EmitterState(Life Cycle Mode=::NewEnumerator1, Loop Behavior=::NewEnumerator1, Loop Duration=1, UseLoopDelay=false)
   &SpawnBurst_Instantaneous(Spawn Count=4, Spawn Time=&Add_Float(A=$User.FirstBeat, B=0), Use Spawn Probability=false, UseLoopCountLimit=false)
  ]
  particleSpawnScript [
   &InitializeParticle(
    Color=<rgba 0.508 0.7018 1 1>
    Lifetime=0.9
    Mesh Scale=&SelectVectorFromArray(
     Direct Array Index=&ReturnExecIndex()
     Vector Selection Array=DataInterface() {propertyValues {FloatData [<1.5 2 1>, <1 1.5 1.3>, <1.8 1 1>]}}
    )
    Mesh Scale Mode=::NewEnumerator4
    UsePositionOffset=false
    Write Material Random=true
   )
   &ShapeLocation(MaskSpawnGroup=false, Override Seed=false, Sphere Radius=10, UseOwnerScale=true) {enabled=false}
   &AddVelocity(
    Cone Angle=60
    Cone Axis=&SelectVectorFromArray(Direct Array Index=$User.SwingSectionIndex, Vector Selection Array=$User.SwingDirectionIce)
    Inner Cone Angle=30
    Module.UseConeFalloff=true
    Module.UseVelDistribution=false
    Override Seed=false
    Velocity Mode=::NewEnumerator2
    Velocity Speed=&FN_FrontEndFloatModifier(Gameplay=&RandomRangeFloat(Minimum=500, Maximum=800), Front End=0)
   ) {enabled=false}
   &AddVelocity(
    Module.Use Falloff Distance=false
    Override Seed=false
    Velocity Mode=::NewEnumerator1
    Velocity Speed=&FN_FrontEndFloatModifier(Gameplay=0, Front End=&RandomRangeFloat(Minimum=200, Maximum=350))
   ) {enabled=false, module=AddVelocity001}
   &InitialMeshOrientation(
    Mesh Axis To Orient=<-0.6574 0.6111 -1>
    Mesh Orientation Mode=::NewEnumerator2
    Rotation=&MakeVector(
     X=&RandomRangeFloat(Minimum=0.1, Maximum=-0.1)
     Y=&SelectFloatFromArray(
      Direct Array Index=&ReturnExecIndex()
      Float Selection Array=DataInterface() {propertyValues {FloatData [-0.0921,0,0.0716,0]}}
     )
     Z=0
    )
    Use Rotation Vector=true
   )
   &CameraOffset()
  ]
  particleUpdateScript [
   &ParticleState()
   &SkeletalMeshLocation(
    Bone / Socket Index=1
    Bone Sampling Mode=::NewEnumerator4
    MaskSpawnGroup=false
    Sampled Orientation Coordinate Space=::World
    Sampled Position Offset=<-5.0036 1.5436 0>
    Skeletal Mesh=$User.Skeletal Mesh
   )
   &Drag(
    Drag=&FloatFromCurve(
     FloatCurve=DataInterface() {propertyValues {Curve {keys [
         {arriveTangent=0.4391, interpMode=RCIM_Cubic, leaveTangent=0.4391, tangentMode=RCTM_User, time=0, value=0}
         {interpMode=RCIM_Cubic, tangentMode=RCTM_User, time=1, value=1.175}
        ]}}}
     Scale Curve=&RandomRangeFloat(Minimum=7, Maximum=9)
    )
    Use Linear Drag=true
    Use Rotational Drag=true
   ) {enabled=false}
   &ScaleColor(
    Scale Alpha=&FloatFromCurve(FloatCurve=DataInterface() {propertyValues {Curve {keys [<at 0.5 1.0113 RCIM_Cubic>,<at 1.0011 0.002 RCIM_Cubic>]}}})
    Scale RGB=&VectorFromFloat(Value=&Multiply_Float(
      A=&RandomRangeFloat(Minimum=1, Maximum=2)
      B=&FNTimeOfDayScalar(Night Scale=0.5, Morning Scale=0.75, Evening Scale=0.75)
     ))
    ScaleA=true
    ScaleRGB=true
   )
   &ScaleMeshSize(Scale Factor=&VectorFromFloat(Value=&FloatFromCurve(
      FloatCurve=DataInterface() {propertyValues {Curve {keys [
          {arriveTangent=8.2917, interpMode=RCIM_Cubic, leaveTangent=8.2917, tangentMode=RCTM_User, time=0, value=0}
          {arriveTangent=0.0045, interpMode=RCIM_Cubic, leaveTangent=0.0045, tangentMode=RCTM_User, time=0.3, value=1}
         ]}}}
      CurveIndex=&Subtract_Float(A=$Particles.NormalizedAge, B=&ReturnNormalizedExecIndex(Normalized Index Scale=0.5))
      Scale Curve=&FN_FrontEndFloatModifier(Gameplay=0.5, Front End=0.8)
     )))
   &ScaleMeshSize(Scale Factor=&VectorFromFloat(Value=&FloatFromCurve(FloatCurve=DataInterface() {propertyValues {Curve {keys [<at 0.8 1>,<at 1 0>]}}}))) {module=ScaleMeshSize001}
   &SolveForcesAndVelocity(Clamp Velocity=false, Limit Acceleration=false)
   /SetVariables(
    Particles.Linear Color=&MultiplyLinearColor(
     Linear Color=$Particles.Color
     Scale Factor=&MakeVector4(
      X=1
      Y=1
      Z=1
      W=&Multiply_Float(
       A=&FloatFromCurve(FloatCurve=DataInterface() {propertyValues {Curve {keys [{interpMode=RCIM_Cubic, tangentMode=RCTM_User, time=0.2, value=0}, {interpMode=RCIM_Cubic, tangentMode=RCTM_User, time=1, value=1}]}}})
       B=&FNTimeOfDayScalar(Night Scale=0.8, Morning Scale=1.5, Evening Scale=0.8)
      )
     )
    )
    Particles.NewPosition:NiagaraPosition=&AddVectorToPosition(
     Position=$Particles.Position
     Vector=&RotateVectorByQuaternion(Vector To Rotate=<-10.1379 4.6131 -6.631>, Quaternion=$Particles.SkeletalMeshLocation.SampledRotation)
    )
   ) {isSetParameters=true}
  ]
  rendererClasses [&NiagaraMeshRendererProperties, &NiagaraSpriteRendererProperties]
  renderers [
   {rendererClass=&NiagaraMeshRendererProperties, rendererIndex=0}
   {rendererClass=&NiagaraMeshRendererProperties, rendererIndex=1}
   {rendererClass=&NiagaraMeshRendererProperties, rendererIndex=2}
   {rendererClass=&NiagaraSpriteRendererProperties, rendererIndex=3}
  ]
 }
 NE_ToonShadingMesh001 {
  emitterSpawnScript [
   &NM_ToonShaderParameters()
   &NM_OutlineParameters(OutlineColor:LinearColor=<rgba 0 0.0601 0.224 1>)
  ]
  emitterUpdateScript [
   &EmitterState()
   &SpawnBurst_Instantaneous(
    Spawn Count=4
    Spawn Time=&Add_Float(A=$User.SecondBeat, B=-0.15)
    Use Spawn Probability=false
    UseLoopCountLimit=false
   )
  ]
  particleSpawnScript [
   &InitializeParticle(
    Color=<rgba 0.508 0.7018 1 1>
    Lifetime=&RandomRangeFloat(Minimum=0.3, Maximum=0.6)
    Mesh Scale=&SelectVectorFromArray(
     Direct Array Index=&ReturnExecIndex()
     Vector Selection Array=DataInterface() {propertyValues {FloatData [<1.5 2 1>, <1 1.5 1.3>, <1.8 1 1>]}}
    )
    Mesh Scale Mode=::NewEnumerator4
    UsePositionOffset=false
    Write Material Random=true
   )
   &SkeletalMeshLocation(
    Bone / Socket Index=1
    Bone Sampling Mode=::NewEnumerator4
    MaskSpawnGroup=false
    Sampled Orientation Coordinate Space=::World
    Sampled Position Offset=<-5.0036 1.5436 0>
    Skeletal Mesh=$User.Skeletal Mesh
   )
   &ShapeLocation(MaskSpawnGroup=false, Override Seed=false, Ring Radius=10, Shape Primitive=::NewEnumerator4, UseOwnerScale=true)
   &AddVelocity(
    Cone Angle=120
    Cone Axis=<0 0 -1>
    Inner Cone Angle=30
    Module.UseConeFalloff=true
    Module.UseVelDistribution=false
    Override Seed=false
    Rotation Coordinate Space=::World
    Velocity Mode=::NewEnumerator2
    Velocity Speed=&FN_FrontEndFloatModifier(Gameplay=&RandomRangeFloat(Minimum=100, Maximum=100), Front End=0)
   ) {enabled=false}
   &AddVelocity(
    Module.Use Falloff Distance=false
    Override Seed=false
    Velocity Mode=::NewEnumerator1
    Velocity Speed=&FN_FrontEndFloatModifier(Gameplay=0, Front End=&RandomRangeFloat(Minimum=200, Maximum=350))
   ) {enabled=false, module=AddVelocity001}
   &InitialMeshOrientation(
    Mesh Axis To Orient=<-0.6574 0.6111 -1>
    Mesh Orientation Mode=::NewEnumerator2
    Orientation Coordinate Space=::Local
    Rotation=&MakeVector(
     X=&RandomRangeFloat(Minimum=0.1, Maximum=-0.1)
     Y=&SelectFloatFromArray(
      Direct Array Index=&ReturnExecIndex()
      Float Selection Array=DataInterface() {propertyValues {FloatData [-0.0921,0,0.0716,0]}}
     )
     Z=0
    )
    Use Rotation Vector=true
   ) {enabled=false}
   &AddVelocity(
    Cone Angle=40
    Cone Axis=<0 0 1>
    Module.UseConeFalloff=false
    Module.UseVelDistribution=false
    Override Seed=false
    Velocity Mode=::NewEnumerator2
    Velocity Speed=250
   ) {module=AddVelocity002}
   &CameraOffset()
  ]
  particleUpdateScript [
   &ParticleState()
   &Drag(
    Drag=&FloatFromCurve(
     FloatCurve=DataInterface() {propertyValues {Curve {keys [
         {arriveTangent=0.4391, interpMode=RCIM_Cubic, leaveTangent=0.4391, tangentMode=RCTM_User, time=0, value=0}
         {interpMode=RCIM_Cubic, tangentMode=RCTM_User, time=1, value=1.175}
        ]}}}
     Scale Curve=&RandomRangeFloat(Minimum=7, Maximum=9)
    )
    Use Linear Drag=true
    Use Rotational Drag=true
   ) {enabled=false}
   &ScaleColor(
    Scale Alpha=&FloatFromCurve(FloatCurve=DataInterface() {propertyValues {Curve {keys [<at 0.5 1.0113 RCIM_Cubic>,<at 1.0011 0.002 RCIM_Cubic>]}}})
    Scale RGB=&VectorFromFloat(Value=&Multiply_Float(
      A=&RandomRangeFloat(Minimum=1, Maximum=2)
      B=&FNTimeOfDayScalar(Night Scale=0.5, Morning Scale=0.75, Evening Scale=0.75)
     ))
    ScaleA=true
    ScaleRGB=true
   )
   &ScaleMeshSize(Scale Factor=&VectorFromFloat(Value=&FloatFromCurve(
      FloatCurve=DataInterface() {propertyValues {Curve {keys [
          {arriveTangent=8.2917, interpMode=RCIM_Cubic, leaveTangent=8.2917, tangentMode=RCTM_User, time=0, value=0}
          {arriveTangent=0.0045, interpMode=RCIM_Cubic, leaveTangent=0.0045, tangentMode=RCTM_User, time=0.3, value=1}
         ]}}}
      CurveIndex=&Subtract_Float(A=$Particles.NormalizedAge, B=&ReturnNormalizedExecIndex(Normalized Index Scale=0.5))
      Scale Curve=&FN_FrontEndFloatModifier(Gameplay=0.2, Front End=0.8)
     )))
   &ScaleMeshSize(Scale Factor=&VectorFromFloat(Value=&FloatFromCurve(FloatCurve=DataInterface() {propertyValues {Curve {keys [<at 0.2 1>,<at 1 0>]}}}))) {module=ScaleMeshSize001}
   &ScaleMeshSize(Scale Factor=&VectorFromFloat(Value=&FloatFromCurve(
      FloatCurve=DataInterface() {propertyValues {Curve {keys [
          {arriveTangent=8.2917, interpMode=RCIM_Cubic, leaveTangent=8.2917, tangentMode=RCTM_User, time=0, value=0}
          {arriveTangent=0.0045, interpMode=RCIM_Cubic, leaveTangent=0.0045, tangentMode=RCTM_User, time=0.3, value=1}
          {arriveTangent=-1.7958, interpMode=RCIM_Cubic, leaveTangent=-1.7958, tangentMode=RCTM_User, time=1, value=0}
         ]}}}
      Scale Curve=&FN_FrontEndFloatModifier(Gameplay=1.2, Front End=0.8)
     ))) {module=ScaleMeshSize002}
   &GravityForce()
   &SolveForcesAndVelocity(Clamp Velocity=false, Limit Acceleration=false)
  ]
  rendererClasses [&NiagaraMeshRendererProperties]
  renderers [
   {rendererClass=&NiagaraMeshRendererProperties, rendererIndex=0}
   {rendererClass=&NiagaraMeshRendererProperties, rendererIndex=1}
   {rendererClass=&NiagaraMeshRendererProperties, rendererIndex=2}
  ]
 }
 NE_ToonShadingMesh002 {
  emitterSpawnScript [
   &NM_ToonShaderParameters()
   &NM_OutlineParameters(OutlineColor:LinearColor=<rgba 0 0.0601 0.224 1>)
  ]
  emitterUpdateScript [
   &EmitterState()
   &SpawnBurst_Instantaneous(Spawn Time=&Add_Float(A=$User.ThirdBeat, B=0), Use Spawn Probability=false, UseLoopCountLimit=false) {module=SpawnBurst_Instantaneous001}
  ]
  particleSpawnScript [
   &InitializeParticle(
    Color=<rgba 0.508 0.7018 1 1>
    Lifetime=$User.LastBeatLifeTime
    Mesh Scale=&SelectVectorFromArray(
     Direct Array Index=&ReturnExecIndex()
     Vector Selection Array=DataInterface() {propertyValues {FloatData [<1.5 2 1>, <1 1.5 1.3>, <1.8 1 1>]}}
    )
    Mesh Scale Mode=::NewEnumerator4
    UsePositionOffset=false
    Write Material Random=true
   )
   &ShapeLocation(MaskSpawnGroup=false, Override Seed=false, Sphere Radius=10, UseOwnerScale=true) {enabled=false}
   &AddVelocity(
    Cone Angle=60
    Cone Axis=&SelectVectorFromArray(Direct Array Index=$User.SwingSectionIndex, Vector Selection Array=$User.SwingDirectionIce)
    Inner Cone Angle=30
    Module.UseConeFalloff=true
    Module.UseVelDistribution=false
    Override Seed=false
    Velocity Mode=::NewEnumerator2
    Velocity Speed=&FN_FrontEndFloatModifier(Gameplay=&RandomRangeFloat(Minimum=500, Maximum=800), Front End=0)
   ) {enabled=false}
   &AddVelocity(
    Module.Use Falloff Distance=false
    Override Seed=false
    Velocity Mode=::NewEnumerator1
    Velocity Speed=&FN_FrontEndFloatModifier(Gameplay=0, Front End=&RandomRangeFloat(Minimum=200, Maximum=350))
   ) {enabled=false, module=AddVelocity001}
   &InitialMeshOrientation(
    Mesh Axis To Orient=<0 0 0>
    Mesh Orientation Mode=::NewEnumerator2
    Rotation=&MakeVector(
     X=&RandomRangeFloat(Minimum=0.1, Maximum=-0.1)
     Y=&SelectFloatFromArray(
      Direct Array Index=&ReturnExecIndex()
      Float Selection Array=DataInterface() {propertyValues {FloatData [-0.0921,0,0.0716,0]}}
     )
     Z=0
    )
    Use Rotation Vector=true
   ) {enabled=false}
   &InitialMeshOrientation(
    Mesh Axis To Orient=<0.3918 -0.1312 -1>
    Mesh Orientation Mode=::NewEnumerator2
    Rotation=&MakeVector(
     X=&RandomRangeFloat(Minimum=0.1, Maximum=-0.1)
     Y=&SelectFloatFromArray(
      Direct Array Index=&ReturnExecIndex()
      Float Selection Array=DataInterface() {propertyValues {FloatData [-0.0921,0,0.0716,0]}}
     )
     Z=0
    )
    Use Rotation Vector=true
   ) {module=InitialMeshOrientation001}
   &CameraOffset()
  ]
  particleUpdateScript [
   &ParticleState()
   &SkeletalMeshLocation(
    Bone / Socket Index=1
    Bone Sampling Mode=::NewEnumerator4
    MaskSpawnGroup=false
    Sampled Orientation Coordinate Space=::World
    Sampled Position Offset=<9.2997 1.5436 3.7617>
    Skeletal Mesh=$User.Skeletal Mesh
   )
   &Drag(
    Drag=&FloatFromCurve(
     FloatCurve=DataInterface() {propertyValues {Curve {keys [
         {arriveTangent=0.4391, interpMode=RCIM_Cubic, leaveTangent=0.4391, tangentMode=RCTM_User, time=0, value=0}
         {interpMode=RCIM_Cubic, tangentMode=RCTM_User, time=1, value=1.175}
        ]}}}
     Scale Curve=&RandomRangeFloat(Minimum=7, Maximum=9)
    )
    Use Linear Drag=true
    Use Rotational Drag=true
   ) {enabled=false}
   &ScaleColor(
    Scale Alpha=&FloatFromCurve(FloatCurve=DataInterface() {propertyValues {Curve {keys [<at 0.5 1.0113 RCIM_Cubic>,<at 1.0011 0.002 RCIM_Cubic>]}}})
    Scale RGB=&VectorFromFloat(Value=&Multiply_Float(
      A=&Multiply_Float(
       A=&RandomRangeFloat(Minimum=1, Maximum=2)
       B=&FNTimeOfDayScalar(Night Scale=0.5, Morning Scale=0.75, Evening Scale=0.75)
      )
      B=&FloatFromCurve(FloatCurve=DataInterface() {propertyValues {Curve {keys [<at 0 8>,<at 1 1>]}}})
     ))
    ScaleA=true
    ScaleRGB=true
   )
   &ScaleMeshSize(Scale Factor=&VectorFromFloat(Value=&FloatFromCurve(
      FloatCurve=DataInterface() {propertyValues {Curve {keys [
          {interpMode=RCIM_Cubic, tangentMode=RCTM_User, time=0, value=0}
          {interpMode=RCIM_Cubic, tangentMode=RCTM_User, time=0.2, value=1}
          <at 1 0.8>
         ]}}}
      Scale Curve=1.2019
     ))) {module=ScaleMeshSize002}
   &ScaleMeshSize(Scale Factor=<0.5 0.5 1>) {module=ScaleMeshSize001}
   &SolveForcesAndVelocity(Clamp Velocity=false, Limit Acceleration=false)
  ]
  rendererClasses [&NiagaraMeshRendererProperties]
  renderers [
   {rendererClass=&NiagaraMeshRendererProperties, rendererIndex=0}
   {rendererClass=&NiagaraMeshRendererProperties, rendererIndex=1}
   {rendererClass=&NiagaraMeshRendererProperties, rendererIndex=2}
   {rendererClass=&NiagaraMeshRendererProperties, rendererIndex=3}
  ]
 }
 Ribbons_L {
  emitterUpdateScript [
   &EmitterState(
    Life Cycle Mode=::NewEnumerator1
    Loop Behavior=::NewEnumerator1
    Loop Delay=&Add_Float(A=$User.SecondBeat, B=0.23)
    UseLoopDelay=true
   )
   &SpawnRate(SpawnRate=100, Use Spawn Probability=false)
  ]
  particleSpawnScript [
   &InitializeParticle(
    Color=<rgba 0.8 0.8 0.8 1>
    Lifetime=$User.TrailLifetime
    Ribbon Width=15
    Ribbon Width Mode=::NewEnumerator1
    Sprite Size Mode=::NewEnumerator3
    Uniform Sprite Size=50
    UsePositionOffset=false
    Write Material Random=true
   )
   &SkeletalMeshLocation(Bone Sampling Mode=::NewEnumerator4, MaskSpawnGroup=false, Skeletal Mesh=$User.Skeletal Mesh)
  ]
  particleUpdateScript [
   &ParticleState()
   &ScaleColor(
    Scale Alpha=&FloatFromCurve(FloatCurve=DataInterface() {propertyValues {Curve {keys [<at -0.0046 1.0113 RCIM_Cubic>,<at 1.0011 0.002 RCIM_Cubic>]}}})
    ScaleA=true
    ScaleRGB=false
   ) {module=ScaleColor002}
   &ScaleColor(
    Scale Alpha=&FloatFromCurve(
     FloatCurve=DataInterface() {propertyValues {Curve {keys [{interpMode=RCIM_Cubic, tangentMode=RCTM_User, time=0.1, value=1}, {interpMode=RCIM_Cubic, tangentMode=RCTM_User, time=0.26, value=0}]}}}
     CurveIndex=$Emitter.LoopedAge
    )
    ScaleA=true
    ScaleRGB=true
   )
  ]
  rendererClasses [&NiagaraRibbonRendererProperties]
  renderers [
   {rendererClass=&NiagaraRibbonRendererProperties}
   {rendererClass=&NiagaraRibbonRendererProperties, rendererIndex=1}
  ]
 }
 Ribbons_R {
  emitterUpdateScript [
   &EmitterState(
    Life Cycle Mode=::NewEnumerator1
    Loop Behavior=::NewEnumerator1
    Loop Delay=&Add_Float(A=$User.SecondBeat, B=0.25)
    Loop Duration=$User.TrailLifetime
    UseLoopDelay=true
   )
   &SpawnRate(SpawnRate=100, Use Spawn Probability=false)
  ]
  particleSpawnScript [
   &InitializeParticle(
    Color=<rgba 0.4091 0.5991 1 1>
    Lifetime=0.2
    Ribbon Width=15
    Ribbon Width Mode=::NewEnumerator1
    Sprite Size Mode=::NewEnumerator3
    Uniform Sprite Size=50
    UsePositionOffset=false
    Write Material Random=true
   )
   &SkeletalMeshLocation(Bone / Socket Index=1, Bone Sampling Mode=::NewEnumerator4, MaskSpawnGroup=false, Skeletal Mesh=$User.Skeletal Mesh)
   &DynamicMaterialParameters(
    Index 0 Param 1=&SelectFloatFromArray(
     Direct Array Index=&SpawnGroup()
     Float Selection Array=DataInterface() {propertyValues {FloatData [-0.3,-0.5,-0.4]}}
    )
    Module.Index0Param0Write=true
    Module.Index0Param1Write=true
    Module.Index0Param2Write=true
    Module.Index0Param3Write=true
   )
  ]
  particleUpdateScript [
   &ParticleState()
   &ScaleColor(
    Scale Alpha=&FloatFromCurve(FloatCurve=DataInterface() {propertyValues {Curve {keys [<at -0.0046 1.0113 RCIM_Cubic>,<at 1.0011 0.002 RCIM_Cubic>]}}}, Scale Curve=10)
    ScaleA=true
    ScaleRGB=false
   ) {module=ScaleColor002}
   &ScaleColor(
    Scale Alpha=&FloatFromCurve(FloatCurve=DataInterface() {propertyValues {Curve {keys [<at 0.7 1>,<at 0.9 0>]}}}, CurveIndex=$Emitter.NormalizedLoopAge)
    ScaleA=true
    ScaleRGB=true
   )
   &ScaleColor(
    Scale Alpha=&FloatFromCurve(
     FloatCurve=DataInterface() {propertyValues {Curve {keys [{interpMode=RCIM_Cubic, tangentMode=RCTM_User, time=0.1, value=1}, {interpMode=RCIM_Cubic, tangentMode=RCTM_User, time=0.26, value=0}]}}}
     CurveIndex=$Emitter.LoopedAge
    )
    ScaleA=true
    ScaleRGB=true
   ) {module=ScaleColor001}
  ]
  rendererClasses [&NiagaraRibbonRendererProperties]
  renderers [
   {rendererClass=&NiagaraRibbonRendererProperties}
   {rendererClass=&NiagaraRibbonRendererProperties, rendererIndex=1}
  ]
 }
 Shock_wave001 {
  emitterUpdateScript [
   &EmitterState()
   &SpawnRate(SpawnRate=7, Use Spawn Probability=false) {enabled=false}
   &SpawnBurst_Instantaneous(Spawn Time=&Add_Float(A=$User.ThirdBeat, B=-0.02), Use Spawn Probability=false, UseLoopCountLimit=false) {module=SpawnBurst_Instantaneous001}
  ]
  particleSpawnScript [
   /Niagara/Modules/Spawn/Initialization/InitializeParticle(
    Color=<rgba 1 0.8445 0.1146 2>
    Lifetime=0.3
    Mesh Scale=&RandomRangeVector(Minimum=<2.2 2.7 2.7>, Maximum=<2.4 2.5 2.5>)
    Module.Write Color=true
    Module.Write Lifetime=true
    Module.Write Mass=true
    Module.Write Material Random=true
    Module.Write Position=false
    Module.Write Scale=true
    Module.Write SpriteRotation=false
    Module.Write SpriteSize=true
    Position=&SimulationPosition()
    Sprite Size=<5 5>
   )
   &SystemLocation(MaskSpawnGroup=false, Offset=<0 0 20>)
   &InitialMeshOrientation(Rotation=<0 0.25 0>, Use Orientation Vector=true, Use Rotation Vector=true)
   &InitialMeshOrientation(
    Rotation=&RandomRangeVector(Minimum=<0 0 0>, Maximum=<0 0 0.15>)
    Use Orientation Vector=true
    Use Rotation Vector=true
   ) {module=InitialMeshOrientation001}
  ]
  particleUpdateScript [
   &ParticleState()
   &SkeletalMeshLocation(
    Bone Sampling Mode=::NewEnumerator4
    MaskSpawnGroup=false
    Orientation Sampling=::NewEnumerator1
    Sampled Position Offset=<0 0 20>
    Skeletal Mesh=DataInterface() {propertyValues {
      DefaultMesh=null
      ExcludeBoneName=root
      OptionalComponentClass=null
      PreviewMesh=&M_Med_Soldier_04_Preview
      SkinningMode=SkinOnTheFly
      SoftSourceActor=null
      SourceMode=Default
      UvSetIndex=0
      WholeMeshLOD=-1
      bExcludeBone=false
      bReadDeformedGeometry=true
      bRequireCurrentFrameData=true
      ComponentTags []
      FilteredBones [root]
      FilteredSockets []
      MeshUserParameter {parameter {name=None, typeDefHandle {registeredTypeIndex=1}, varData []}}
      SamplingRegions []
     }}
   )
   &SolveForcesAndVelocity(Clamp Velocity=false, Limit Acceleration=false)
   &ScaleColor(
    Scale Alpha=&Multiply_Float(
     A=1
     B=&Multiply_Float(
      A=&FloatFromCurve(FloatCurve=DataInterface() {propertyValues {Curve {keys [<at 0 1 RCIM_Cubic>,<at 1 0>]}}})
      B=&FNTimeOfDayScalar(Night Scale=0.5, Morning Scale=0.2, Evening Scale=0.2)
     )
    )
    ScaleA=true
    ScaleRGB=true
   ) {module=ScaleColor001}
   &ScaleMeshSize(Scale Factor=&VectorFromCurve(
     VectorCurve=DataInterface() {propertyValues {XCurve {keys [<at 0 0.5>,<at 1 1.5>]}, YCurve {keys [<at 0 0.7>,<at 1 1.6>]}, ZCurve {keys [<at 0 0.7>,<at 1 2>]}}}
     Scale Curve=<0.7 0.7 0.7>
    ))
   &DynamicMaterialParameters(
    Index 0 Param 1=&FloatFromCurve(FloatCurve=DataInterface() {propertyValues {Curve {keys [<at 0 0 RCIM_Cubic>,<at 1 0.28 RCIM_Cubic>]}}})
    Index 0 Param 2=&RandomRangeFloat(Minimum=0, Maximum=1)
    Index 0 Param 3=50
    Index 0 Param 4=&FloatFromCurve(FloatCurve=DataInterface() {propertyValues {Curve {keys [<at 0 0>, {arriveTangent=-3.0374, leaveTangent=2.0371, time=0.2, value=0}, {interpMode=RCIM_Cubic, tangentMode=RCTM_Break, time=1, value=0}]}}})
    Module.Param0WriteEnabled=true
    Module.Param1WriteEnabled=true
    Module.Param2WriteEnabled=true
    Module.Param3WriteEnabled=true
   )
   /SetVariables(Particles.New Position:NiagaraPosition=$Particles.Position) {enabled=false, isSetParameters=true}
  ]
  rendererClasses [&NiagaraMeshRendererProperties]
  renderers [{rendererClass=&NiagaraMeshRendererProperties, rendererIndex=0}]
 }
 SnowParticles {
  emitterUpdateScript [
   &EmitterState()
   &SpawnBurst_Instantaneous(
    Spawn Count=6
    Spawn Time=&Add_Float(A=$User.ThirdBeat, B=-0.02)
    Use Spawn Probability=false
    UseLoopCountLimit=false
   )
   &SpawnRate(
    Spawn Probability=&MakeCustomFloatFromBool(Bool=$User.bisSwinging, True Float=0, False Float=1)
    SpawnRate=5
    Use Spawn Probability=true
   ) {enabled=false}
  ]
  particleSpawnScript [
   /Niagara/Modules/Spawn/Initialization/V2/InitializeParticle(
    Color=&RandomRangeLinearColor(Minimum=<rgba 1 1 1 1>, Maximum=$User.SnowBlue)
    Lifetime=&RandomRangeFloat(Minimum=0.15, Maximum=0.4)
    Sprite Rotation Mode=::NewEnumerator2
    Sprite Size Mode=::NewEnumerator1
    Sprite UV Mode=::NewEnumerator3
    UsePositionOffset=false
    Write Material Random=true
   )
   &SkeletalMeshLocation(
    Bone / Socket Index=1
    Bone Sampling Mode=::NewEnumerator4
    MaskSpawnGroup=false
    Sampled Orientation Coordinate Space=::World
    Sampled Position Offset=<37.6666 -21.5486 2.7769>
    Skeletal Mesh=$User.Skeletal Mesh
   ) {module=SkeletalMeshLocation001}
   &ShapeLocation(
    MaskSpawnGroup=false
    Non Uniform Scale=<2 2.1519 2>
    Offset=<-16.7931 19.3502 12.9363>
    Offset Coordinate Space=::Local
    Offset Mode=::NewEnumerator0
    Override Seed=false
    Rotation Angle=0.3455
    Rotation Axis=<0 0 1>
    Rotation Mode=::NewEnumerator1
    Sphere Radius=5
    UseOwnerScale=true
   )
   &AddVelocity(
    Cone Angle=97.8828
    Cone Axis=<-1 0.3426 0>
    Inner Cone Angle=20
    Module.UseConeFalloff=false
    Module.UseVelDistribution=false
    Override Seed=false
    Velocity Mode=::NewEnumerator2
    Velocity Speed=&Multiply_Float(A=0.5593, B=&RandomRangeFloat(Minimum=700, Maximum=1200))
   )
  ]
  particleUpdateScript [
   &ParticleState()
   &ScaleSpriteSize(
    Uniform Curve Scale=4
    Uniform Curve Sprite Scale=DataInterface() {propertyValues {Curve {keys [
        {arriveTangent=9.0725, interpMode=RCIM_Cubic, leaveTangent=9.0725, tangentMode=RCTM_User, time=0, value=0}
        {arriveTangent=-2.5249, interpMode=RCIM_Cubic, leaveTangent=-2.5249, tangentMode=RCTM_User, time=0.05, value=0.9847}
        {arriveTangent=-0.0405, interpMode=RCIM_Cubic, leaveTangent=-0.0405, tangentMode=RCTM_User, time=1, value=0}
       ]}}}
   )
   &ScaleColor(
    Scale Alpha=&FloatFromCurve(FloatCurve=DataInterface() {propertyValues {Curve {keys [
         {interpMode=RCIM_Cubic, tangentMode=RCTM_User, time=0, value=0}
         {interpMode=RCIM_Cubic, tangentMode=RCTM_User, time=0.5, value=1}
         {tangentMode=RCTM_User, time=1, value=0}
        ]}}})
    Scale RGB=&VectorFromFloat(Value=&FNTimeOfDayScalar(Night Scale=0.8, Morning Scale=1, Evening Scale=0.8))
    ScaleA=true
    ScaleRGB=true
   )
   &ScaleColor(Color Value To Scale=$Particles.Color, Scale Alpha=$System.Timeline.InverseNormalizedTime, ScaleA=true, ScaleRGB=false) {enabled=false, module=ScaleColor001}
   &WindForce(
    Frequency=60
    Offset By Time:Vector3f=<0 0 0>
    Scale=1.5
    Wind Speed=<0 0 0>
    Wind Speed Scale=&Lerp_Float(A=20, B=30, Alpha=$System.MovingSpeed)
   )
   &AerodynamicDrag(
    Aerodynamic Drag:NiagaraFloat=&Multiply_Float(A=0.5, B=&RandomRangeFloat(Minimum=0.2, Maximum=0.5))
    Aerodynamic Rotational Drag=0
    Drag Shape Curve=DataInterface() {propertyValues {Curve {keys [
        {interpMode=RCIM_Cubic, tangentMode=RCTM_User, time=0, value=0.1}
        {arriveTangent=2.0799, interpMode=RCIM_Cubic, leaveTangent=2.0799, tangentMode=RCTM_User, time=1, value=1}
       ]}}}
    Lift Contribution=0
    Lift Shape Curve=DataInterface() {propertyValues {Curve {keys [
        {interpMode=RCIM_Cubic, tangentMode=RCTM_User, time=0, value=0}
        {interpMode=RCIM_Cubic, tangentMode=RCTM_User, time=0.5, value=1}
        {tangentMode=RCTM_User, time=1, value=0}
       ]}}}
    Use Linear Drag=true
    Use Rotational Drag=true
   )
   &CameraOffset(Camera Offset Amount=10) {enabled=false}
   &GravityForce(Gravity=<0 0 -5>)
   &InheritSourceMovement(
    Applied Position Scale=&VectorFromFloat(Value=&FloatFromCurve(FloatCurve=DataInterface() {propertyValues {Curve {keys [
          {arriveTangent=-2.0944, interpMode=RCIM_Cubic, leaveTangent=-2.0944, tangentMode=RCTM_User, time=0, value=0.85}
          {interpMode=RCIM_Cubic, tangentMode=RCTM_User, time=0.5, value=0}
         ]}}}))
    ApplyForce=false
    ApplyPosition=true
    ApplyPositionFalloff=false
   ) {enabled=false}
   &CurlNoiseForce(Noise Strength=1, Pan Noise=false)
   &GravityForce(Gravity=<0 0 -200>) {module=GravityForce001}
   &SolveForcesAndVelocity(Clamp Velocity=false, Limit Acceleration=false)
   &ScaleSpriteSizeBySpeed(Max Scale Factor=<1 6>, Velocity Threshold=20)
   &ScaleColor(Scale Alpha=$System.FadeOut, ScaleA=true, ScaleRGB=true) {enabled=false, module=ScaleColor002}
  ]
  rendererClasses [&NiagaraSpriteRendererProperties, &NiagaraMeshRendererProperties]
  renderers [
   {rendererClass=&NiagaraSpriteRendererProperties}
   {rendererClass=&NiagaraMeshRendererProperties, rendererIndex=1}
  ]
 }
 line001 {
  emitterUpdateScript [
   &EmitterState()
   &SpawnRate(SpawnRate=20, Use Spawn Probability=false) {enabled=false}
   &SpawnBurst_Instantaneous(
    Spawn Count=4
    Spawn Time=&Add_Float(A=$User.ThirdBeat, B=0.02)
    Use Spawn Probability=false
    UseLoopCountLimit=false
   ) {module=SpawnBurst_Instantaneous001}
  ]
  particleSpawnScript [
   &InitializeParticle(
    Color=<rgba 0 0 0 1>
    Lifetime=&RandomRangeFloat(Minimum=0.1, Maximum=0.15)
    Position Mode=::NewEnumerator1
    Position Offset=<0 0 20>
    Sprite Size Max=<80 150>
    Sprite Size Min=<20 50>
    Sprite Size Mode=::NewEnumerator2
    UsePositionOffset=true
    Write Material Random=true
   )
   &SystemLocation(MaskSpawnGroup=false, Offset=<0 0 20>) {enabled=false}
   &ShapeLocation(Handle Radius=0, Large Radius=40, MaskSpawnGroup=false, Override Seed=false, Shape Primitive=::NewEnumerator3, UseOwnerScale=true)
   &AddVelocity(
    Module.Use Falloff Distance=false
    Origin Offset=<0 0 -200>
    Override Seed=false
    Velocity Mode=::NewEnumerator1
    Velocity Origin=&MakeVector(X=0, Y=0, Z=0)
    Velocity Speed=&RandomRangeFloat(Minimum=1000, Maximum=1500)
   )
   &AddVelocity(Velocity=<0 0 100>) {module=AddVelocity001}
  ]
  particleUpdateScript [
   &ParticleState()
   &Drag(Drag=5, Use Linear Drag=true, Use Rotational Drag=false)
   &SolveForcesAndVelocity(Clamp Velocity=false, Limit Acceleration=false)
   &ScaleColor(
    Scale Alpha=&FloatFromCurve(
     FloatCurve=DataInterface() {propertyValues {Curve {keys [
         {interpMode=RCIM_Cubic, tangentMode=RCTM_User, time=0, value=0}
         {interpMode=RCIM_Cubic, tangentMode=RCTM_User, time=0.5, value=1}
         {tangentMode=RCTM_User, time=1, value=0}
        ]}}}
     Scale Curve=2
    )
    ScaleA=true
    ScaleRGB=true
   )
   &ScaleSpriteSize(
    Non-Uniform Curve Sprite Scale=DataInterface() {propertyValues {
      XCurve {keys [{tangentMode=RCTM_User, time=0, value=0.5}, {tangentMode=RCTM_User, time=1, value=1.5}]}
      YCurve {keys [{tangentMode=RCTM_User, time=0, value=1}, {tangentMode=RCTM_User, time=1, value=0.5}]}
     }}
    Scale Sprite Size Mode=::NewEnumerator3
   )
   /SetVariables(Particles.New Position:NiagaraPosition=&Add_Vector(
     A=Linked() {linkedVariable {name=Particles.Position, type=NiagaraPosition}}
     B=Linked() {linkedVariable {name=Engine.Emitter.SimulationPosition, type=NiagaraPosition}}
    )) {isSetParameters=true}
  ]
  rendererClasses [&NiagaraSpriteRendererProperties]
  renderers [{rendererClass=&NiagaraSpriteRendererProperties}]
 }
}
summary {
 systemName=NS_Pickaxe_SteamPowerOrange_Emote_Anticipation1
 emitters [
  {
   emitterName=Fire_Mesh_Cross_Plane
   rendererClasses [&NiagaraMeshRendererProperties, &NiagaraSpriteRendererProperties]
  }
  {
   emitterName=NE_ToonShadingMesh
   rendererClasses [&NiagaraMeshRendererProperties, &NiagaraSpriteRendererProperties]
  }
  {
   emitterName=Fire_Mesh_Cross_Plane001
   rendererClasses [
    &NiagaraMeshRendererProperties
    &NiagaraLightRendererProperties
    &NiagaraSpriteRendererProperties
   ]
  }
  {emitterName=Embers, rendererClasses [&NiagaraSpriteRendererProperties, &NiagaraMeshRendererProperties]}
  {emitterName=Ribbons_L, rendererClasses [&NiagaraRibbonRendererProperties]}
  {emitterName=Ribbons_R, rendererClasses [&NiagaraRibbonRendererProperties]}
  {emitterName=NE_ToonShadingMesh001, rendererClasses [&NiagaraMeshRendererProperties]}
  {emitterName=NE_ToonShadingMesh002, rendererClasses [&NiagaraMeshRendererProperties]}
  {emitterName=line001, rendererClasses [&NiagaraSpriteRendererProperties]}
  {emitterName=Shock_wave001, rendererClasses [&NiagaraMeshRendererProperties]}
  {emitterName=AirGlow003, rendererClasses [&NiagaraSpriteRendererProperties]}
  {
   emitterName=Flare_Glow
   rendererClasses [
    &NiagaraMeshRendererProperties
    &NiagaraSpriteRendererProperties
    &NiagaraLightRendererProperties
   ]
  }
  {emitterName=SnowParticles, rendererClasses [&NiagaraSpriteRendererProperties, &NiagaraMeshRendererProperties]}
  {
   emitterName=Flare_and_Light
   rendererClasses [&NiagaraMeshRendererProperties, &NiagaraSpriteRendererProperties]
  }
  {
   emitterName=Flare_and_Light001
   rendererClasses [&NiagaraMeshRendererProperties, &NiagaraSpriteRendererProperties]
  }
 ]
 userVariables [
  User.AnimState {type=NiagaraFloat, defaultValue=1, description=""}
  User.bIsFrontEnd {type=NiagaraBool}
  User.bIsFrontEndPreview {type=NiagaraBool}
  User.bisSwinging {type=NiagaraBool}
  User.FirstBeat {type=NiagaraFloat, defaultValue=0.35, description=""}
  User.FrontendScale {type=NiagaraFloat, defaultValue=1, description=""}
  User.LastBeatLifeTime {type=NiagaraFloat, defaultValue=0.12, description=""}
  User.NormalizedNotifyProgress {type=NiagaraFloat, defaultValue=0, description=""}
  User.RibbonIDOffset {type=NiagaraInt32}
  User.SecondBeat {type=NiagaraFloat, defaultValue=1.25, description=""}
  User.Skeletal Mesh {
   type=NiagaraDataInterfaceSkeletalMesh
   defaultValue {
    struct=NiagaraExt_VariableValue_DataInterface
    value {dataInterface=&self:NiagaraDataInterfaceSkeletalMesh_0}
   }
  }
  User.SnowBlue {type=LinearColor, description="", defaultValue {struct=LinearColor, value=<rgba 0.2581 0.5474 0.901 1>}}
  User.SwingDirectionFire {
   type=NiagaraDataInterfaceArrayFloat3
   description=""
   defaultValue {
    struct=NiagaraExt_VariableValue_DataInterface
    value {
     dataInterface=&self:NiagaraDataInterfaceArrayFloat3_0
     dataInterfaceClass=&NiagaraDataInterfaceArrayFloat3
    }
   }
  }
  User.SwingDirectionIce {
   type=NiagaraDataInterfaceArrayFloat3
   description=""
   defaultValue {
    struct=NiagaraExt_VariableValue_DataInterface
    value {
     dataInterface=&self:NiagaraDataInterfaceArrayFloat3_1
     dataInterfaceClass=&NiagaraDataInterfaceArrayFloat3
    }
   }
  }
  User.SwingSectionIndex {type=NiagaraInt32, defaultValue=3}
  User.ThirdBeat {type=NiagaraFloat, defaultValue=1.72, description=""}
  User.TrailLifetime {type=NiagaraFloat, defaultValue=0.55, description=""}
 ]
}
