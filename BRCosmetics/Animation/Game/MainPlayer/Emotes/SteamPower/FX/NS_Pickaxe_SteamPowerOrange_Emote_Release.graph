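# Niagara release burst for the orange SteamPower pickaxe emote, timed to land just before the end of its last beat.
# ShardsBurst throws pale-blue, toon-outlined mesh shards and Toon_Smoke adds a brown sprite puff, both placed by skeletal mesh sampling.
# Spawn times derive from User.ThirdBeat and LastBeatLifeTime; color dims by time of day, and shard size adapts to front end.
# AirGlow, Embers001 and Flare_and_Light emitters are disabled.
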NiagaraSystem /BRCosmetics/Animation/Game/MainPlayer/Emotes/SteamPower/FX/NS_Pickaxe_SteamPowerOrange_Emote_Release
dependencies {
 usedDataInterfaces [
  &NiagaraDataInterfaceCurve
  &NiagaraDataInterfaceColorCurve
  &NiagaraDataInterfacePlatformSet
  &NiagaraDataInterfaceStaticMesh
  &NiagaraDataInterfaceArrayFloat
 ]
 usedDynamicInputs []
 usedModules [
  &self:SystemSpawnScript.SystemScriptSource.SystemScriptGraph.NiagaraNodeAssignment_3.SetVariables_#1
  &self:SystemSpawnScript.SystemScriptSource.SystemScriptGraph.NiagaraNodeAssignment_1.SetVariables_#2
  &SystemState
  &self:SystemSpawnScript.SystemScriptSource.SystemScriptGraph.NiagaraNodeAssignment_2.SetVariables_#3
 ]
}
emitters {
 AirGlow {
  bEnabled=false
  emitterUpdateScript [
   &EmitterState()
   &SpawnBurst_Instantaneous(Use Spawn Probability=false, UseLoopCountLimit=false)
  ]
  particleSpawnScript [&InitializeParticle(
    Color=<rgba 3 0.25 0 0.5>
    Lifetime=2
    Sprite Size=<30 120>
    Sprite Size Mode=::NewEnumerator4
    UsePositionOffset=true
    Write Material Random=true
   )]
  particleUpdateScript [
   &ParticleState()
   &SkeletalMeshLocation(
    Bone Sampling Mode=::NewEnumerator4
    MaskSpawnGroup=false
    Sampled Position Offset=<40 0 0>
    Sampled Position Offset Coordinate Space:ENiagaraSkelMeshTransforms=::NewEnumerator4
    Skeletal Mesh=$System.Skeletal Mesh
    Use Sampled Mesh Attribute as Ribbon ID=false
   )
   &CameraOffset(Camera Offset Amount=15)
   /SetVariables(Particles.Velocity=&RotateVectorByQuaternion(Vector To Rotate=<1 0 0>, Quaternion=$Particles.SkeletalMeshLocation.SampledRotation)) {isSetParameters=true}
   &ScaleColor(
    Scale RGB=&VectorFromFloat(Value=&FNTimeOfDayScalar(Night Scale=0.5, Morning Scale=0.75, Evening Scale=0.75))
    ScaleA=true
    ScaleRGB=true
   )
   &ScaleColor(
    Scale Alpha=&Lerp_Float(A=$System.EmoteIntro, B=$System.EmoteOutro, Alpha=$User.AnimState)
    ScaleA=true
    ScaleRGB=true
   ) {module=ScaleColor002}
  ]
  rendererClasses [&NiagaraSpriteRendererProperties, &NiagaraMeshRendererProperties]
  renderers [
   {rendererClass=&NiagaraSpriteRendererProperties}
   {rendererClass=&NiagaraMeshRendererProperties, rendererIndex=1}
  ]
 }
 Embers001 {
  bEnabled=false
  emitterUpdateScript [
   &EmitterState()
   &SpawnRate(
    Spawn Probability=&MakeCustomFloatFromBool(Bool=$User.bisSwinging, True Float=0, False Float=0.5)
    SpawnRate=400
    Use Spawn Probability=true
   )
  ]
  particleSpawnScript [
   /SetVariables(Particles.PerParticleRandom:NiagaraFloat=&RandomRangeFloat(Minimum=0, Maximum=1)) {isSetParameters=true}
   &InitializeParticle(
    Color Mode=::NewEnumerator0
    Lifetime=&Lerp_Float(A=0.75, B=0.3, Alpha=$Particles.PerParticleRandom)
    Sprite Size=&Multiply_Vector2D(
     A=&Multiply_Vector2D_ByFloat(
      Float=&FN_FrontEndFloatModifier(Gameplay=1, Front End=&Multiply_Float(A=$User.FrontendScale, B=1))
      Vector2D=<0.5 1>
     )
     B=&MakeVector2D(
      X=&FloatFromCurve(
       FloatCurve=DataInterface() {propertyValues {Curve {keys [<at 0 1>,<at 0.8 2 RCIM_Constant>,<at 0.95 3 RCIM_Constant>]}}}
       CurveIndex=$Particles.PerParticleRandom
      )
      Y=&FloatFromCurve(
       FloatCurve=DataInterface() {propertyValues {Curve {keys [<at 0 1>,<at 0.8 2 RCIM_Constant>,<at 0.95 3>]}}}
       CurveIndex=$Particles.PerParticleRandom
      )
     )
    )
    Sprite Size Mode=::NewEnumerator4
    UsePositionOffset=false
    Write Material Random=true
   )
   &ShapeLocation(
    Cylinder Height=&Multiply_Float(A=60, B=&FN_FrontEndFloatModifier(Gameplay=1, Front End=$User.FrontendScale))
    Cylinder Height Midpoint=0
    Cylinder Radius=3
    MaskSpawnGroup=false
    Override Seed=false
    Rotation Axis=<0 0.2 1>
    Rotation Mode=::NewEnumerator1
    Shape Primitive=::NewEnumerator1
    Surface Only=false
    UseOwnerScale=true
   ) {module=ShapeLocation001}
   &AddVelocity(
    Cone Angle=90
    Cone Axis=<0 0 1>
    Module.UseConeFalloff=true
    Module.UseVelDistribution=false
    Override Seed=false
    Velocity Mode=::NewEnumerator2
    Velocity Speed=&Multiply_Float(
     A=&RandomRangeFloat(Minimum=150, Maximum=300)
     B=&FN_FrontEndFloatModifier(Gameplay=1, Front End=$User.FrontendScale)
    )
   ) {module=AddVelocity001}
   &SkeletalMeshLocation(Bone Sampling Mode=::NewEnumerator4, MaskSpawnGroup=false, Sampled Position Offset=$System.Offset, Skeletal Mesh=$System.Skeletal Mesh)
  ]
  particleUpdateScript [
   &ParticleState()
   &ScaleSpriteSize(Uniform Curve Sprite Scale=DataInterface() {propertyValues {Curve {keys [{tangentMode=RCTM_User, time=0, value=1}, {tangentMode=RCTM_User, time=1, value=1}]}}}) {module=ScaleSpriteSize001}
   &Drag(
    Drag=&FloatFromCurve(
     FloatCurve=DataInterface() {propertyValues {Curve {keys [
         {interpMode=RCIM_Cubic, tangentMode=RCTM_User, time=0, value=0.2}
         {arriveTangent=2.3593, interpMode=RCIM_Cubic, leaveTangent=2.3593, tangentMode=RCTM_User, time=1, value=1}
        ]}}}
     Scale Curve=&RandomRangeFloat(Minimum=30, Maximum=50)
    )
    Use Linear Drag=true
    Use Rotational Drag=false
   )
   &CurlNoiseForce(
    Noise Strength=&FloatFromCurve(
     FloatCurve=DataInterface() {propertyValues {Curve {keys [{interpMode=RCIM_Cubic, tangentMode=RCTM_User, time=0, value=0.4}, {interpMode=RCIM_Cubic, tangentMode=RCTM_User, time=1, value=1}]}}}
     Scale Curve=500
    )
    Pan Noise=false
    Random Seed=33
   ) {module=CurlNoiseForce001}
   &InheritSourceMovement(
    Applied Position Scale=&VectorFromFloat(Value=&FloatFromCurve(FloatCurve=DataInterface() {propertyValues {Curve {keys [{interpMode=RCIM_Cubic, tangentMode=RCTM_User, time=0, value=1}, {interpMode=RCIM_Cubic, tangentMode=RCTM_User, time=1, value=0}]}}}))
    ApplyForce=false
    ApplyPosition=true
    ApplyPositionFalloff=false
   )
   &SolveForcesAndVelocity(Clamp Velocity=false, Limit Acceleration=false)
   &ScaleColor(
    Scale RGB=&MakeVectorFromLinearColorRGB(Color=&ColorFromCurve(ColorCurve=DataInterface() {propertyValues {
        AlphaCurve {keys [<at 0 1 RCIM_Cubic>]}
        BlueCurve {keys [<at 0 0 RCIM_Cubic>,<at 0.2325 0>,<at 1 0 RCIM_Cubic>]}
        GreenCurve {keys [<at 0 0.5 RCIM_Cubic>,<at 0.2325 0>,<at 1 0 RCIM_Cubic>]}
        RedCurve {keys [<at 0 10 RCIM_Cubic>,<at 0.2325 1>,<at 1 0.5 RCIM_Cubic>]}
       }}))
    ScaleA=true
    ScaleRGB=true
   )
   &ScaleSpriteSize(
    Uniform Curve Scale=&MakeFloatFromVector(Vector=$Engine.Owner.Scale, Channel=::NewEnumerator0)
    Uniform Curve Sprite Scale=DataInterface() {propertyValues {Curve {keys [
        {tangentMode=RCTM_User, time=0, value=1}
        {arriveTangent=-0.0405, interpMode=RCIM_Cubic, leaveTangent=-0.0405, tangentMode=RCTM_User, time=0.7009, value=1}
        {arriveTangent=-8.0089, interpMode=RCIM_Cubic, leaveTangent=-8.0089, tangentMode=RCTM_User, time=1, value=0}
       ]}}}
   )
   &ScaleSpriteSizeBySpeed(
    Max Scale Factor=<1.35 5>
    Min Scale Factor=<1.333 1.333>
    Sample Scale Factor By Curve=true
    Scale Factor Curve=DataInterface() {propertyValues {Curve {keys [<at 0 0>, {arriveTangent=1.4989, interpMode=RCIM_Cubic, tangentMode=RCTM_Break, time=1, value=1}]}}}
    Velocity Threshold=250
   )
   &ScaleColor(
    Scale Alpha=&Lerp_Float(A=$System.EmoteIntro, B=$System.EmoteOutro, Alpha=$User.AnimState)
    ScaleA=true
    ScaleRGB=true
   ) {module=ScaleColor002}
  ]
  rendererClasses [&NiagaraSpriteRendererProperties]
  renderers [{rendererClass=&NiagaraSpriteRendererProperties}]
 }
 Flare_and_Light {
  bEnabled=false
  emitterUpdateScript [
   &EmitterState()
   &SpawnBurst_Instantaneous(Use Spawn Probability=false, UseLoopCountLimit=false)
  ]
  particleSpawnScript [&InitializeParticle(Color=<rgba 5 1.25 0 1>, UsePositionOffset=false, Write Material Random=true)]
  particleUpdateScript [
   &ParticleState()
   &SkeletalMeshLocation(Bone Sampling Mode=::NewEnumerator4, MaskSpawnGroup=false, Sampled Position Offset=$System.Offset, Skeletal Mesh=$System.Skeletal Mesh)
   &ScaleColor(
    Scale RGB=&VectorFromFloat(Value=&FNTimeOfDayScalar(Night Scale=0.5, Morning Scale=0.75, Evening Scale=0.75))
    ScaleA=true
    ScaleRGB=true
   ) {module=ScaleColor001}
   &ScaleColor(
    Scale Alpha=&Lerp_Float(A=$System.EmoteIntro, B=$System.EmoteOutro, Alpha=$User.AnimState)
    ScaleA=true
    ScaleRGB=true
   ) {module=ScaleColor002}
  ]
  rendererClasses [&NiagaraLightRendererProperties]
  renderers [{rendererClass=&NiagaraLightRendererProperties, rendererIndex=0}]
 }
 ShardsBurst {
  emitterSpawnScript [
   /Game/Effects/Niagara/MasterMaterials/ToonShader/Template/NM_ToonShaderParameters()
   /Game/Effects/Niagara/MasterMaterials/ToonShader/Template/NM_OutlineParameters(OutlineColor:LinearColor=<rgba 0 0.0601 0.224 1>)
  ]
  emitterUpdateScript [
   &EmitterState()
   &SpawnBurst_Instantaneous(
    Spawn Count=7
    Spawn Time=&Add_Float(A=$User.ThirdBeat, B=&Subtract_Float(A=$User.LastBeatLifeTime, B=0.05))
    Use Spawn Probability=false
    UseLoopCountLimit=false
   )
  ]
  particleSpawnScript [
   &InitializeParticle(
    Color=<rgba 0.508 0.7018 1 1>
    Lifetime=&RandomRangeFloat(Minimum=0.3, Maximum=0.6)
    Mesh Scale Max=<0.7 1 0.7>
    Mesh Scale Min=<1.5 2 1>
    Mesh Scale Mode=::NewEnumerator2
    UsePositionOffset=false
    Write Material Random=true
   )
   &PartitionParticles(Particles Per Partition=4, Partition Mode:ENiagara_PartitionMode=::NewEnumerator1)
   &AddVelocity(
    Cone Angle=120
    Cone Axis=<0 0 -1>
    Inner Cone Angle=30
    Module.UseConeFalloff=true
    Module.UseVelDistribution=false
    Override Seed=false
    Rotation Coordinate Space=::World
    Velocity Mode=::NewEnumerator2
    Velocity Speed=&FN_FrontEndFloatModifier(Gameplay=&RandomRangeFloat(Minimum=100, Maximum=100), Front End=0)
   ) {enabled=false}
   &AddVelocity(
    Module.Use Falloff Distance=false
    Override Seed=false
    Velocity Mode=::NewEnumerator1
    Velocity Speed=&FN_FrontEndFloatModifier(Gameplay=0, Front End=&RandomRangeFloat(Minimum=200, Maximum=350))
   ) {enabled=false, module=AddVelocity001}
   &SkeletalMeshLocation(
    Bone / Socket Index=1
    Bone Sampling Mode=::NewEnumerator4
    MaskSpawnGroup=false
    Sampled Orientation Coordinate Space=::World
    Sampled Position Offset=<-19.7836 8.3037 -6.0214>
    Skeletal Mesh=$User.Skeletal Mesh
   )
   &ShapeLocation(
    MaskSpawnGroup=false
    Offset Coordinate Space=::Local
    Offset Mode=::NewEnumerator0
    Override Seed=false
    Rotation Angle=1.1873
    Rotation Axis=<0 0 1>
    Rotation Coordinate Space=::Simulation
    Rotation Mode=::NewEnumerator1
    Sphere Radius=5
    UseOwnerScale=true
   )
   /Niagara/Modules/Spawn/Location/StaticMeshLocation(
    MaskSpawnGroup=false
    OffsetAlongNormal=false
    OverrideRandomSeed=false
    SamplePosOffset=true
    Sampled Position Offset:Vector3f=<0 0 80>
    Static Mesh=DataInterface() {propertyValues {
      DefaultMesh=&/SM_SteamPowerAnticipation
      InstanceIndex=-1
      LODIndex=0
      PreviewMesh=&SM_SteamPowerAnticipation
      SoftSourceActor=null
      SourceMode=DefaultMeshOnly
      bAllowSamplingFromStreamingLODs=false
      bCaptureTransformsPerFrame=true
      bUsePhysicsBodyVelocity=false
      FilteredSockets []
      LODIndexUserParameter {parameter {name=None, typeDefHandle {registeredTypeIndex=108}, varData []}}
      MeshParameterBinding {parameter {name=None, typeDefHandle {registeredTypeIndex=1}, varData []}}
      SectionFilter {allowedMaterialSlots []}
     }}
   ) {enabled=false}
   &InitialMeshOrientation(
    Mesh Axis To Orient=<-0.3611 0.6111 -1>
    Mesh Orientation Mode=::NewEnumerator2
    Rotation=&MakeVector(
     X=&RandomRangeFloat(Minimum=0.1, Maximum=-0.1)
     Y=&SelectFloatFromArray(
      Direct Array Index=&ReturnExecIndex()
      Float Selection Array=DataInterface() {propertyValues {FloatData [-0.0921,0,0.0716,0]}}
     )
     Z=0
    )
    Use Rotation Vector=true
   )
   &AddVelocity(
    Module.Use Falloff Distance=false
    Override Seed=false
    Velocity Mode=::NewEnumerator1
    Velocity Origin=$Particles.Position
    Velocity Speed=200
   ) {module=AddVelocity002}
   &AddVelocity(Velocity=&Lerp_Vector(
     A=<0 0 400>
     B=<-166.0942 0 -8.8124>
     Alpha=&MakeFloatFromInt(Int=$Particles.PartitionParticles.Partition)
    )) {module=AddVelocity003}
   &CameraOffset()
  ]
  particleUpdateScript [
   &ParticleState()
   &Drag(
    Drag=&FloatFromCurve(
     FloatCurve=DataInterface() {propertyValues {Curve {keys [
         {arriveTangent=0.4391, interpMode=RCIM_Cubic, leaveTangent=0.4391, tangentMode=RCTM_User, time=0, value=0}
         {interpMode=RCIM_Cubic, tangentMode=RCTM_User, time=1, value=1.175}
        ]}}}
     Scale Curve=&RandomRangeFloat(Minimum=7, Maximum=9)
    )
    Use Linear Drag=true
    Use Rotational Drag=true
   ) {enabled=false}
   &ScaleColor(
    Scale Alpha=&FloatFromCurve(FloatCurve=DataInterface() {propertyValues {Curve {keys [<at 0.5 1.0113 RCIM_Cubic>,<at 1.0011 0.002 RCIM_Cubic>]}}})
    Scale RGB=&VectorFromFloat(Value=&Multiply_Float(
      A=&RandomRangeFloat(Minimum=1, Maximum=2)
      B=&FNTimeOfDayScalar(Night Scale=0.5, Morning Scale=0.75, Evening Scale=0.75)
     ))
    ScaleA=true
    ScaleRGB=true
   )
   &ScaleMeshSize(Scale Factor=&VectorFromFloat(Value=&FloatFromCurve(
      FloatCurve=DataInterface() {propertyValues {Curve {keys [{arriveTangent=8.2917, interpMode=RCIM_Cubic, leaveTangent=8.2917, tangentMode=RCTM_User, time=0, value=0.7}]}}}
      Scale Curve=&FN_FrontEndFloatModifier(Gameplay=0.1, Front End=0.2)
     )))
   &ScaleMeshSize(Scale Factor=&VectorFromFloat(Value=&FloatFromCurve(FloatCurve=DataInterface() {propertyValues {Curve {keys [<at 0.2 1>,<at 1 0>]}}}, Scale Curve=4))) {module=ScaleMeshSize001}
   &ScaleMeshSize(Scale Factor=&VectorFromFloat(Value=&FloatFromCurve(
      FloatCurve=DataInterface() {propertyValues {Curve {keys [
          {arriveTangent=8.2917, interpMode=RCIM_Cubic, leaveTangent=8.2917, tangentMode=RCTM_User, time=0, value=0}
          {arriveTangent=0.0045, interpMode=RCIM_Cubic, leaveTangent=0.0045, tangentMode=RCTM_User, time=0.15, value=1}
          {arriveTangent=-1.7958, interpMode=RCIM_Cubic, leaveTangent=-1.7958, tangentMode=RCTM_User, time=1, value=0}
         ]}}}
      Scale Curve=&FN_FrontEndFloatModifier(Gameplay=1.2, Front End=0.8)
     ))) {module=ScaleMeshSize002}
   &ScaleMeshSize(Scale Factor=&VectorFromFloat(Value=&FloatFromCurve(FloatCurve=DataInterface() {propertyValues {Curve {keys [<at 0 0 RCIM_Constant>,<at 0.2 1 RCIM_Constant>]}}}))) {enabled=false, module=ScaleMeshSize003}
   &GravityForce()
   &Drag(Drag=8, Use Linear Drag=true, Use Rotational Drag=true) {module=Drag001}
   &SolveForcesAndVelocity(Clamp Velocity=false, Limit Acceleration=false)
  ]
  rendererClasses [&NiagaraMeshRendererProperties]
  renderers [
   {rendererClass=&NiagaraMeshRendererProperties, rendererIndex=0}
   {rendererClass=&NiagaraMeshRendererProperties, rendererIndex=1}
   {rendererClass=&NiagaraMeshRendererProperties, rendererIndex=2}
   {rendererClass=&NiagaraMeshRendererProperties, rendererIndex=3}
  ]
 }
 Toon_Smoke {
  emitterUpdateScript [
   &EmitterState(Life Cycle Mode=::NewEnumerator1, Loop Behavior=::NewEnumerator1, Loop Delay=0.075, Loop Duration=2, UseLoopDelay=true)
   &SpawnBurst_Instantaneous(
    Spawn Count=1
    Spawn Probability=&FN_FrontEndFloatModifier(Gameplay=1, Front End=1)
    Spawn Time=&Add_Float(A=$User.ThirdBeat, B=&Subtract_Float(A=$User.LastBeatLifeTime, B=0.08))
    Use Spawn Probability=true
   )
  ]
  particleSpawnScript [
   /SetVariables(Particles.Random=&RandomRangeFloat(Minimum=0, Maximum=1)) {isSetParameters=true}
   &PartitionParticles(Particles Per Partition=1, Partition Mode:ENiagara_PartitionMode=::NewEnumerator1)
   &InitializeParticle(
    Color=&LerpLinearColors(StartColor=<rgba 0.4 0.2667 0.2392 1>, EndColor=<rgba 0.3021 0.129 0.0933 1>, LerpFactor=1)
    Lifetime=&RandomRangeFloat(Minimum=0.6, Maximum=0.7)
    Sprite Rotation Mode=::NewEnumerator2
    Sprite Size Max=<170 220>
    Sprite Size Min=<150 200>
    Sprite Size Mode=::NewEnumerator2
    Sprite UV Mode=::NewEnumerator1
    UsePositionOffset=false
    Write Material Random=true
   ) {module=InitializeParticle001}
   &CameraOffset(Camera Offset Amount=30)
   &SkeletalMeshLocation(
    Bone Sampling Mode=::NewEnumerator4
    MaskSpawnGroup=false
    Sampled Position Offset=&Lerp_Vector(A=<-5.3263 -0.9471 -12.5019>, B=<50 0 -5>, Alpha=$Particles.NormalizedAge)
    Skeletal Mesh=$System.Skeletal Mesh
   ) {module=SkeletalMeshLocation001}
   &AddVelocity(Velocity=<0.4241 1 0.5>, Velocity Speed Scale=200)
   &DynamicMaterialParameters(
    Index 0 Param 2=&UniformAOrBFloat(A=0)
    Module.Param0WriteEnabled=true
    Module.Param1WriteEnabled=true
    Module.Param2WriteEnabled=true
    Module.Param3WriteEnabled=true
   ) {module=DynamicMaterialParameters001}
  ]
  particleUpdateScript [
   &ParticleState()
   &ScaleSpriteSize(Scale Factor=<0.4009 1>, Scale Sprite Size Mode=::NewEnumerator2)
   &Drag(
    Drag=&FloatFromCurve(
     FloatCurve=DataInterface() {propertyValues {Curve {keys [{interpMode=RCIM_Cubic, tangentMode=RCTM_User, time=0, value=0}, {interpMode=RCIM_Cubic, tangentMode=RCTM_User, time=0.4, value=1}]}}}
     Scale Curve=10
    )
    Use Linear Drag=true
    Use Rotational Drag=true
   )
   &SolveForcesAndVelocity(Clamp Velocity=false, Limit Acceleration=false)
   &DynamicMaterialParameters(
    Index 0 Param 1=&FloatFromCurve(FloatCurve=DataInterface() {propertyValues {Curve {keys [
         {arriveTangent=1.3143, interpMode=RCIM_Cubic, leaveTangent=1.3143, tangentMode=RCTM_User, time=0, value=0}
         {arriveTangent=0.4262, interpMode=RCIM_Cubic, leaveTangent=0.4262, tangentMode=RCTM_User, time=1, value=1}
        ]}}})
    Module.Param0WriteEnabled=true
    Module.Param1WriteEnabled=false
    Module.Param2WriteEnabled=true
    Module.Param3WriteEnabled=true
   )
   &ScaleColor(
    Scale Alpha=&FloatFromCurve(
     FloatCurve=DataInterface() {propertyValues {Curve {keys [{interpMode=RCIM_Cubic, tangentMode=RCTM_User, time=0, value=1}, {interpMode=RCIM_Cubic, tangentMode=RCTM_User, time=1, value=0}]}}}
     Scale Curve=0.8
    )
    Scale RGB=&VectorFromFloat(Value=&FNTimeOfDayScalar(Night Scale=0.35, Morning Scale=0.75, Evening Scale=0.75))
    ScaleA=true
    ScaleRGB=true
   )
  ]
  rendererClasses [&NiagaraSpriteRendererProperties]
  renderers [{rendererClass=&NiagaraSpriteRendererProperties}]
 }
}
summary {
 systemName=NS_Pickaxe_SteamPowerOrange_Emote_Release
 emitters [
  {
   bEnabled=false
   emitterName=AirGlow
   rendererClasses [&NiagaraSpriteRendererProperties, &NiagaraMeshRendererProperties]
  }
  {bEnabled=false, emitterName=Flare_and_Light, rendererClasses [&NiagaraLightRendererProperties]}
  {bEnabled=false, emitterName=Embers001, rendererClasses [&NiagaraSpriteRendererProperties]}
  {emitterName=ShardsBurst, rendererClasses [&NiagaraMeshRendererProperties]}
  {emitterName=Toon_Smoke, rendererClasses [&NiagaraSpriteRendererProperties]}
 ]
 userVariables [
  User.AnimState {type=NiagaraFloat, defaultValue=1, description=""}
  User.bIsFrontEnd {type=NiagaraBool}
  User.bIsFrontEndPreview {type=NiagaraBool}
  User.bisSwinging {type=NiagaraBool}
  User.FirstBeat {type=NiagaraFloat, defaultValue=0.75, description=""}
  User.FrontendScale {type=NiagaraFloat, defaultValue=1, description=""}
  User.LastBeatLifeTime {type=NiagaraFloat, defaultValue=0.07, description=""}
  User.NormalizedNotifyProgress {type=NiagaraFloat, defaultValue=0, description=""}
  User.RibbonIDOffset {type=NiagaraInt32}
  User.SecondBeat {type=NiagaraFloat, defaultValue=0.75, description=""}
  User.Skeletal Mesh {
   type=NiagaraDataInterfaceSkeletalMesh
   defaultValue {
    struct=NiagaraExt_VariableValue_DataInterface
    value {dataInterface=&self:NiagaraDataInterfaceSkeletalMesh_0}
   }
  }
  User.SnowBlue {type=LinearColor, description="", defaultValue {struct=LinearColor, value=<rgba 0.2581 0.5474 0.901 1>}}
  User.SwingDirectionFire {
   type=NiagaraDataInterfaceArrayFloat3
   description=""
   defaultValue {
    struct=NiagaraExt_VariableValue_DataInterface
    value {
     dataInterface=&self:NiagaraDataInterfaceArrayFloat3_0
     dataInterfaceClass=&NiagaraDataInterfaceArrayFloat3
    }
   }
  }
  User.SwingDirectionIce {
   type=NiagaraDataInterfaceArrayFloat3
   description=""
   defaultValue {
    struct=NiagaraExt_VariableValue_DataInterface
    value {
     dataInterface=&self:NiagaraDataInterfaceArrayFloat3_1
     dataInterfaceClass=&NiagaraDataInterfaceArrayFloat3
    }
   }
  }
  User.SwingSectionIndex {type=NiagaraInt32, defaultValue=3}
  User.ThirdBeat {type=NiagaraFloat, defaultValue=0.8, description=""}
  User.TrailLifetime {type=NiagaraFloat, defaultValue=0.55, description=""}
 ]
}
